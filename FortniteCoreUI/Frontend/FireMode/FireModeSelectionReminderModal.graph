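# Frontend modal that reminds the player they can change weapon fire mode, asking Yes or No, Thanks and noting settings can change it later.
# Shows Autofire, Dedicated Button and Tap Anywhere art over a blurred lightbox, with a touch-only close button.
# Construct plays the staggered ShowUp animation; Cancel and touch close play it in reverse.
# The native FireModeSelectionReminderModal parent supplies the three inherited buttons; this graph leaves Confirm unhandled.
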WidgetBlueprint /FortniteCoreUI/Frontend/FireMode/FireModeSelectionReminderModal
animations [ShowUp {
  binding_count=8
  length_seconds=0.65
  object=&self:ShowUp
  sequence {bindings [
    Containment {
     id=#1
     possessed_class=Image
     tracks [{
       type=MovieScene2DTransformTrack
       sections [{
         type=MovieScene2DTransformSection
         channels [
          Scale.X {type=MovieSceneScriptingFloatChannel}
          Scale.Y {type=MovieSceneScriptingFloatChannel, keys [[1,0],[6,1]]}
         ]
        }]
      }]
    }
    Image_BackgroundBottomBar {
     id=#2
     possessed_class=Image
     tracks [{
       type=MovieScene2DTransformTrack
       sections [{
         type=MovieScene2DTransformSection
         channels [
          Scale.X {type=MovieSceneScriptingFloatChannel}
          Scale.Y {type=MovieSceneScriptingFloatChannel, keys [[1,0],[6,1]]}
         ]
        }]
      }]
    }
    artContainer {
     id=#3
     possessed_class=Overlay
     tracks [{
       type=MovieScene2DTransformTrack
       sections [{
         type=MovieScene2DTransformSection
         channels [
          Scale.X {type=MovieSceneScriptingFloatChannel}
          Scale.Y {type=MovieSceneScriptingFloatChannel, keys [[1,0],[6,1]]}
         ]
        }]
      }]
    }
    Button_Cancel {
     id=#4
     possessed_class=GenericModalButton_C
     tracks [{
       type=MovieScene2DTransformTrack
       sections [{
         type=MovieScene2DTransformSection
         channels [
          Translation.X {type=MovieSceneScriptingFloatChannel, const [0,1]}
          Translation.Y {type=MovieSceneScriptingFloatChannel, const [0,1]}
          Angle {type=MovieSceneScriptingFloatChannel}
          Scale.X {type=MovieSceneScriptingFloatChannel, keys [[0,0],[3,0],[5,1],[9,1]]}
          Scale.Y {type=MovieSceneScriptingFloatChannel, keys [[0,0],[3,0],[5,1],[9,1]]}
          Shear.X {type=MovieSceneScriptingFloatChannel}
          Shear.Y {type=MovieSceneScriptingFloatChannel}
         ]
        }]
      }]
    }
    Title {
     id=#5
     possessed_class=CommonRichTextBlock
     tracks [
      {
       type=MovieSceneFloatTrack
       display_name=Render Opacity
       sections [{type=MovieSceneFloatSection, end_frame=13, start_frame=0, channels [None {type=MovieSceneScriptingFloatChannel, keys [[0,0],[4,0],[8,1]]}]}]
      }
      {
       type=MovieScene2DTransformTrack
       sections [{
         type=MovieScene2DTransformSection
         channels [
          Translation.X {type=MovieSceneScriptingFloatChannel}
          Translation.Y {type=MovieSceneScriptingFloatChannel, keys [[2,100],[8,0]]}
          Angle {type=MovieSceneScriptingFloatChannel}
          Scale.X {type=MovieSceneScriptingFloatChannel}
          Scale.Y {type=MovieSceneScriptingFloatChannel}
          Shear.X {type=MovieSceneScriptingFloatChannel}
          Shear.Y {type=MovieSceneScriptingFloatChannel}
         ]
        }]
      }
     ]
    }
    MobileClose {
     id=#6
     possessed_class=SafeZone
     tracks [
      {
       type=MovieSceneFloatTrack
       display_name=Render Opacity
       sections [{type=MovieSceneFloatSection, end_frame=13, start_frame=0, channels [None {type=MovieSceneScriptingFloatChannel, keys [[0,0],[2,0],[7,1]]}]}]
      }
      {
       type=MovieScene2DTransformTrack
       sections [{
         type=MovieScene2DTransformSection
         channels [
          Translation.X {type=MovieSceneScriptingFloatChannel, const [0,1]}
          Translation.Y {type=MovieSceneScriptingFloatChannel, keys [[2,100],[7,0]]}
          Angle {type=MovieSceneScriptingFloatChannel, const [0,1]}
          Scale.X {type=MovieSceneScriptingFloatChannel, const [1,1]}
          Scale.Y {type=MovieSceneScriptingFloatChannel, const [1,1]}
          Shear.X {type=MovieSceneScriptingFloatChannel, const [0,1]}
          Shear.Y {type=MovieSceneScriptingFloatChannel, const [0,1]}
         ]
        }]
      }
     ]
    }
    Button_Confirm {
     id=#7
     possessed_class=GenericModalButton_C
     tracks [{
       type=MovieScene2DTransformTrack
       sections [{
         type=MovieScene2DTransformSection
         channels [
          Translation.X {type=MovieSceneScriptingFloatChannel, const [0,1]}
          Translation.Y {type=MovieSceneScriptingFloatChannel, const [0,1]}
          Angle {type=MovieSceneScriptingFloatChannel}
          Scale.X {type=MovieSceneScriptingFloatChannel, keys [[2,0],[5,0],[7,1],[11,1]]}
          Scale.Y {type=MovieSceneScriptingFloatChannel, keys [[2,0],[5,0],[7,1],[11,1]]}
          Shear.X {type=MovieSceneScriptingFloatChannel}
          Shear.Y {type=MovieSceneScriptingFloatChannel}
         ]
        }]
      }]
    }
    Text_Description {
     id=#8
     possessed_class=CommonTextBlock
     tracks [
      {
       type=MovieSceneFloatTrack
       display_name=Render Opacity
       sections [{type=MovieSceneFloatSection, end_frame=13, start_frame=0, channels [None {type=MovieSceneScriptingFloatChannel, keys [[2,0],[6,0],[10,1]]}]}]
      }
      {
       type=MovieScene2DTransformTrack
       sections [{
         type=MovieScene2DTransformSection
         channels [
          Translation.X {type=MovieSceneScriptingFloatChannel}
          Translation.Y {type=MovieSceneScriptingFloatChannel, keys [[4,100],[10,0]]}
          Angle {type=MovieSceneScriptingFloatChannel}
          Scale.X {type=MovieSceneScriptingFloatChannel}
          Scale.Y {type=MovieSceneScriptingFloatChannel}
          Shear.X {type=MovieSceneScriptingFloatChannel}
          Shear.Y {type=MovieSceneScriptingFloatChannel}
         ]
        }]
      }
     ]
    }
   ]}
 }]
graphs [EventGraph {nodes [
   n1 = "On Button Clicked (Button_Cancel)"!() -> (OutputDelegate, then, Button:Common Button Legacy Object Reference) {is=title K2Node_ComponentBoundEvent}
   n2 = Get ShowUp() -> (ShowUp:Widget Animation Object Reference)
   PlayAnimation(execute=[@n1.then, @n3.then], InAnimation=@n2, PlayMode=Reverse)
   n3 = "On Button Clicked (Button_TouchClose)"!() -> (OutputDelegate, then, Button:Common Button Legacy Object Reference) {is=title K2Node_ComponentBoundEvent}
   Event Construct()
   > PlayAnimation(InAnimation=@n2)
  ]}]
variables []
widgets {
 info {inheritedWidgetCount=3, parentClass=/Script/FortniteUI.FireModeSelectionReminderModal, rootWidgetClass=&Overlay, widgetCount=34}
 widgets [
  Overlay {widgetClassPath=&Overlay, widget=Overlay}
  BackgroundBlurLightbox {
   widgetClassPath=&BackgroundBlur
   parent=Overlay
   widget=BackgroundBlurLightbox
   properties {
    blurStrength=8
    lowQualityFallbackBrush {
     bIsDynamicallyLoaded=false
     drawAs=Image
     imageType=NoImage
     mirroring=NoMirror
     resourceName=None
     resourceObject=null
     tiling=NoTile
     imageSize=<0 0>
     margin=<ltrb 0 0 0 0>
     outlineSettings {
      bUseBrushTransparency=false
      roundingType=HalfHeightRadius
      width=0
      color {colorUseRule=UseColor_Specified, specifiedColor=<rgba 0 0 0 0>}
      cornerRadii=<0 0 0 0>
     }
     tintColor {colorUseRule=UseColor_Specified, specifiedColor=<rgba 0 0 0 0.75>}
     uVRegion {bIsValid=false, max=<0 0>, min=<0 0>}
    }
   }
   slot {class=OverlaySlot, properties {horizontalAlignment=HAlign_Fill, verticalAlignment=VAlign_Fill}}
  }
  Lightbox_HitTestBlocker {
   widgetClassPath=&Image
   bIsVariable=true
   parent=BackgroundBlurLightbox
   widget=Lightbox_HitTestBlocker
   properties {brush {
     bIsDynamicallyLoaded=false
     drawAs=Image
     imageType=NoImage
     mirroring=NoMirror
     resourceName=None
     resourceObject=/Game/Athena/UI/Frontend/Art/M_UI_RadialGradientMaster_Dark
     tiling=NoTile
     imageSize=<32 32>
     margin=<ltrb 0 0 0 0>
     outlineSettings {
      bUseBrushTransparency=false
      roundingType=HalfHeightRadius
      width=0
      color {colorUseRule=UseColor_Specified, specifiedColor=<rgba 0 0 0 0>}
      cornerRadii=<0 0 0 0>
     }
     tintColor {colorUseRule=UseColor_Specified, specifiedColor=<rgba 0.5972 0.5972 0.5972 1>}
     uVRegion {bIsValid=false, max=<0 0>, min=<0 0>}
    }}
   slot {class=BackgroundBlurSlot, properties {padding=<ltrb 0 0 0 0>}}
  }
  PrimaryHeightSB {
   widgetClassPath=&SizeBox
   parent=Overlay
   widget=PrimaryHeightSB
   properties {bOverride_MinDesiredHeight=true, minDesiredHeight=500}
   slot {class=OverlaySlot, properties {horizontalAlignment=HAlign_Fill, verticalAlignment=VAlign_Center}}
  }
  Overlay_Modal {widgetClassPath=&Overlay, parent=PrimaryHeightSB, widget=Overlay_Modal, slot {class=SizeBoxSlot}}
  Containment {
   widgetClassPath=&Image
   bIsVariable=true
   parent=Overlay_Modal
   widget=Containment
   properties {
    visibility=HitTestInvisible
    brush {
     bIsDynamicallyLoaded=false
     drawAs=Image
     imageType=NoImage
     mirroring=NoMirror
     resourceName=None
     resourceObject=/Game/Athena/UI/Subscriptions/Art/M_UI_Subs_Color_FIll_Modals
     tiling=NoTile
     imageSize=<32 32>
     margin=<ltrb 0 0 0 0>
     outlineSettings {
      bUseBrushTransparency=false
      roundingType=HalfHeightRadius
      width=0
      color {colorUseRule=UseColor_Specified, specifiedColor=<rgba 0 0 0 0>}
      cornerRadii=<0 0 0 0>
     }
     tintColor {colorUseRule=UseColor_Specified, specifiedColor=<rgba 1 1 1 1>}
     uVRegion {bIsValid=false, max=<0 0>, min=<0 0>}
    }
    colorAndOpacity=<rgba 1 1 1 0.5>
    renderTransformPivot=<0.5 1>
   }
   slot {class=OverlaySlot, properties {horizontalAlignment=HAlign_Fill, verticalAlignment=VAlign_Fill}}
  }
  ContentSZ {
   widgetClassPath=&SafeZone
   bIsVariable=true
   parent=Overlay_Modal
   widget=ContentSZ
   slot {class=OverlaySlot, properties {horizontalAlignment=HAlign_Fill, verticalAlignment=VAlign_Fill, padding=<ltrb 64 0 64 0>}}
  }
  HorizontalBox_0 {widgetClassPath=&HorizontalBox, parent=ContentSZ, widget=HorizontalBox_0, slot {class=SafeZoneSlot}}
  artContainer {
   widgetClassPath=&Overlay
   parent=HorizontalBox_0
   widget=artContainer
   slot {class=HorizontalBoxSlot, properties {padding=<ltrb 0 -60 0 40>, size {sizeRule=Fill, value=0.25}}}
  }
  Firemode2 {
   widgetClassPath=&Overlay
   parent=artContainer
   widget=Firemode2
   slot {class=OverlaySlot, properties {horizontalAlignment=HAlign_Center, verticalAlignment=VAlign_Fill, padding=<ltrb -100 0 0 0>}}
  }
  SizeBox {
   widgetClassPath=&SizeBox
   parent=Firemode2
   widget=SizeBox
   properties {bOverride_HeightOverride=true, bOverride_WidthOverride=true, heightOverride=420, widthOverride=260}
   slot {class=OverlaySlot, properties {horizontalAlignment=HAlign_Center, verticalAlignment=VAlign_Center}}
  }
  Firemode01_1 {
   widgetClassPath=&Image
   bIsVariable=true
   parent=SizeBox
   widget=Firemode01_1
   properties {brush {
     bIsDynamicallyLoaded=false
     drawAs=Image
     imageType=NoImage
     mirroring=NoMirror
     resourceName=None
     resourceObject=/FortniteCoreUI/Frontend/FireMode/art/Firemode_RemainderModal/MI_FireMode_DedicatedButton
     tiling=NoTile
     imageSize=<32 32>
     margin=<ltrb 0 0 0 0>
     outlineSettings {
      bUseBrushTransparency=false
      roundingType=HalfHeightRadius
      width=0
      color {colorUseRule=UseColor_Specified, specifiedColor=<rgba 0 0 0 0>}
      cornerRadii=<0 0 0 0>
     }
     tintColor {colorUseRule=UseColor_Specified, specifiedColor=<rgba 1 1 1 1>}
     uVRegion {bIsValid=false, max=<0 0>, min=<0 0>}
    }}
   slot {class=SizeBoxSlot}
  }
  SelectedIcon_1 {
   widgetClassPath=&Image
   bIsVariable=true
   parent=Firemode2
   widget=SelectedIcon_1
   properties {
    visibility=Collapsed
    brush {
     bIsDynamicallyLoaded=false
     drawAs=Image
     imageType=NoImage
     mirroring=NoMirror
     resourceName=None
     resourceObject=/Game/UI/Foundation/Textures/Icons/Checks/T-Icon-Check-64
     tiling=NoTile
     imageSize=<50 50>
     margin=<ltrb 0 0 0 0>
     outlineSettings {
      bUseBrushTransparency=false
      roundingType=HalfHeightRadius
      width=0
      color {colorUseRule=UseColor_Specified, specifiedColor=<rgba 0 0 0 0>}
      cornerRadii=<0 0 0 0>
     }
     tintColor {colorUseRule=UseColor_Specified, specifiedColor=<rgba 0 1 0 1>}
     uVRegion {bIsValid=false, max=<0 0>, min=<0 0>}
    }
    colorAndOpacity=<rgba 1 1 1 0.7>
   }
   slot {class=OverlaySlot, properties {horizontalAlignment=HAlign_Center, verticalAlignment=VAlign_Bottom, padding=<ltrb 0 0 0 15>}}
  }
  Firemode3 {
   widgetClassPath=&Overlay
   parent=artContainer
   widget=Firemode3
   slot {class=OverlaySlot, properties {horizontalAlignment=HAlign_Center, verticalAlignment=VAlign_Fill, padding=<ltrb 110 0 0 0>}}
  }
  SizeBox_1 {
   widgetClassPath=&SizeBox
   parent=Firemode3
   widget=SizeBox_1
   properties {bOverride_HeightOverride=true, bOverride_WidthOverride=true, heightOverride=420, widthOverride=260}
   slot {class=OverlaySlot, properties {horizontalAlignment=HAlign_Center, verticalAlignment=VAlign_Center}}
  }
  Firemode01_2 {
   widgetClassPath=&Image
   bIsVariable=true
   parent=SizeBox_1
   widget=Firemode01_2
   properties {brush {
     bIsDynamicallyLoaded=false
     drawAs=Image
     imageType=NoImage
     mirroring=NoMirror
     resourceName=None
     resourceObject=/FortniteCoreUI/Frontend/FireMode/art/Firemode_RemainderModal/MI_FireMode_tapAnywhere
     tiling=NoTile
     imageSize=<32 32>
     margin=<ltrb 0 0 0 0>
     outlineSettings {
      bUseBrushTransparency=false
      roundingType=HalfHeightRadius
      width=0
      color {colorUseRule=UseColor_Specified, specifiedColor=<rgba 0 0 0 0>}
      cornerRadii=<0 0 0 0>
     }
     tintColor {colorUseRule=UseColor_Specified, specifiedColor=<rgba 1 1 1 1>}
     uVRegion {bIsValid=false, max=<0 0>, min=<0 0>}
    }}
   slot {class=SizeBoxSlot}
  }
  SelectedIcon_2 {
   widgetClassPath=&Image
   bIsVariable=true
   parent=Firemode3
   widget=SelectedIcon_2
   properties {
    visibility=Collapsed
    brush {
     bIsDynamicallyLoaded=false
     drawAs=Image
     imageType=NoImage
     mirroring=NoMirror
     resourceName=None
     resourceObject=&T-Icon-Check-64
     tiling=NoTile
     imageSize=<50 50>
     margin=<ltrb 0 0 0 0>
     outlineSettings {
      bUseBrushTransparency=false
      roundingType=HalfHeightRadius
      width=0
      color {colorUseRule=UseColor_Specified, specifiedColor=<rgba 0 0 0 0>}
      cornerRadii=<0 0 0 0>
     }
     tintColor {colorUseRule=UseColor_Specified, specifiedColor=<rgba 0 1 0 1>}
     uVRegion {bIsValid=false, max=<0 0>, min=<0 0>}
    }
    colorAndOpacity=<rgba 1 1 1 0.7>
   }
   slot {class=OverlaySlot, properties {horizontalAlignment=HAlign_Center, verticalAlignment=VAlign_Bottom, padding=<ltrb 0 0 0 15>}}
  }
  Firemode_01 {
   widgetClassPath=&Overlay
   parent=artContainer
   widget=Firemode_01
   slot {class=OverlaySlot, properties {horizontalAlignment=HAlign_Center, verticalAlignment=VAlign_Fill}}
  }
  SizeBox_0 {
   widgetClassPath=&SizeBox
   parent=Firemode_01
   widget=SizeBox_0
   properties {bOverride_WidthOverride=true, widthOverride=350}
   slot {class=OverlaySlot, properties {horizontalAlignment=HAlign_Center, verticalAlignment=VAlign_Fill}}
  }
  Firemode01 {
   widgetClassPath=&Image
   bIsVariable=true
   parent=SizeBox_0
   widget=Firemode01
   properties {brush {
     bIsDynamicallyLoaded=false
     drawAs=Image
     imageType=NoImage
     mirroring=NoMirror
     resourceName=None
     resourceObject=/FortniteCoreUI/Frontend/FireMode/art/Firemode_RemainderModal/MI_FireMode_Autofire
     tiling=NoTile
     imageSize=<32 32>
     margin=<ltrb 0 0 0 0>
     outlineSettings {
      bUseBrushTransparency=false
      roundingType=HalfHeightRadius
      width=0
      color {colorUseRule=UseColor_Specified, specifiedColor=<rgba 0 0 0 0>}
      cornerRadii=<0 0 0 0>
     }
     tintColor {colorUseRule=UseColor_Specified, specifiedColor=<rgba 1 1 1 1>}
     uVRegion {bIsValid=false, max=<0 0>, min=<0 0>}
    }}
   slot {class=SizeBoxSlot}
  }
  SelectedIcon {
   widgetClassPath=&Image
   bIsVariable=true
   parent=Firemode_01
   widget=SelectedIcon
   properties {
    visibility=Collapsed
    brush {
     bIsDynamicallyLoaded=false
     drawAs=Image
     imageType=NoImage
     mirroring=NoMirror
     resourceName=None
     resourceObject=&T-Icon-Check-64
     tiling=NoTile
     imageSize=<50 50>
     margin=<ltrb 0 0 0 0>
     outlineSettings {
      bUseBrushTransparency=false
      roundingType=HalfHeightRadius
      width=0
      color {colorUseRule=UseColor_Specified, specifiedColor=<rgba 0 0 0 0>}
      cornerRadii=<0 0 0 0>
     }
     tintColor {colorUseRule=UseColor_Specified, specifiedColor=<rgba 0 1 0 1>}
     uVRegion {bIsValid=false, max=<0 0>, min=<0 0>}
    }
    colorAndOpacity=<rgba 1 1 1 0.7>
   }
   slot {class=OverlaySlot, properties {horizontalAlignment=HAlign_Center, verticalAlignment=VAlign_Bottom, padding=<ltrb 0 0 0 15>}}
  }
  Overlay_2 {
   widgetClassPath=&Overlay
   parent=HorizontalBox_0
   widget=Overlay_2
   slot {
    class=HorizontalBoxSlot
    properties {horizontalAlignment=HAlign_Center, verticalAlignment=VAlign_Center, padding=<ltrb 20 130 20 120>, size {sizeRule=Fill, value=0.5}}
   }
  }
  TextContent {
   widgetClassPath=&VerticalBox
   parent=Overlay_2
   widget=TextContent
   slot {class=OverlaySlot, properties {horizontalAlignment=HAlign_Fill, verticalAlignment=VAlign_Fill}}
  }
  Title {
   widgetClassPath=&CommonRichTextBlock
   bIsVariable=true
   parent=TextContent
   widget=Title
   properties {
    autoWrapText=true
    bOverrideDefaultStyle=true
    justification=Center
    text="<text color=\"fff\">DO YOU WANT TO CHANGE</> HOW YOU FIRE <text color=\"fff\">YOUR WEAPON?</>"
    defaultTextStyleOverride {
     overflowPolicy=Clip
     transformPolicy=None
     colorAndOpacity {colorUseRule=UseColor_Specified, specifiedColor=<rgba 0.2159 0.6584 1 1>}
     font {
      bForceMonospaced=false
      bMaterialIsStencil=false
      fontMaterial=null
      fontObject=&Burbank
      letterSpacing=0
      monospacedWidth=1
      size=40
      skewAmount=0
      typefaceFontName=BurbankRegular-Black
      outlineSettings {
       bApplyOutlineToDropShadows=false
       bMiteredCorners=false
       bSeparateFillAlpha=false
       outlineMaterial=null
       outlineSize=0
       outlineColor=<rgba 0 0 0 1>
      }
     }
     highlightColor {colorUseRule=UseColor_Specified, specifiedColor=<rgba 0 0 0 1>}
     highlightShape {
      bIsDynamicallyLoaded=false
      drawAs=Image
      imageType=NoImage
      mirroring=NoMirror
      resourceName=None
      resourceObject=null
      tiling=NoTile
      imageSize=<32 32>
      margin=<ltrb 0 0 0 0>
      outlineSettings {
       bUseBrushTransparency=false
       roundingType=HalfHeightRadius
       width=0
       color {colorUseRule=UseColor_Specified, specifiedColor=<rgba 0 0 0 0>}
       cornerRadii=<0 0 0 0>
      }
      tintColor {colorUseRule=UseColor_Specified, specifiedColor=<rgba 1 1 1 1>}
      uVRegion {bIsValid=false, max=<0 0>, min=<0 0>}
     }
     selectedBackgroundColor {colorUseRule=UseColor_Foreground, specifiedColor=<rgba 1 0 1 1>}
     shadowColorAndOpacity=<rgba 0 0 0 1>
     shadowOffset=<0 0>
     strikeBrush {
      bIsDynamicallyLoaded=false
      drawAs=Image
      imageType=NoImage
      mirroring=NoMirror
      resourceName=None
      resourceObject=null
      tiling=NoTile
      imageSize=<32 32>
      margin=<ltrb 0 0 0 0>
      outlineSettings {
       bUseBrushTransparency=false
       roundingType=HalfHeightRadius
       width=0
       color {colorUseRule=UseColor_Specified, specifiedColor=<rgba 0 0 0 0>}
       cornerRadii=<0 0 0 0>
      }
      tintColor {colorUseRule=UseColor_Specified, specifiedColor=<rgba 1 1 1 1>}
      uVRegion {bIsValid=false, max=<0 0>, min=<0 0>}
     }
     underlineBrush {
      bIsDynamicallyLoaded=false
      drawAs=Image
      imageType=NoImage
      mirroring=NoMirror
      resourceName=None
      resourceObject=null
      tiling=NoTile
      imageSize=<32 32>
      margin=<ltrb 0 0 0 0>
      outlineSettings {
       bUseBrushTransparency=false
       roundingType=HalfHeightRadius
       width=0
       color {colorUseRule=UseColor_Specified, specifiedColor=<rgba 0 0 0 0>}
       cornerRadii=<0 0 0 0>
      }
      tintColor {colorUseRule=UseColor_Specified, specifiedColor=<rgba 1 1 1 1>}
      uVRegion {bIsValid=false, max=<0 0>, min=<0 0>}
     }
    }
   }
   slot {class=VerticalBoxSlot}
  }
  Text_Description {
   widgetClassPath=&CommonTextBlock
   bIsVariable=true
   parent=TextContent
   widget=Text_Description
   properties {
    autoWrapText=true
    justification=Center
    text=You can always change it later in settings
    wrapTextAt=1152
    colorAndOpacity {colorUseRule=UseColor_Specified, specifiedColor=<rgba 0.0235 0.5922 0.8157 1>}
    font {
     bForceMonospaced=false
     bMaterialIsStencil=false
     fontMaterial=null
     fontObject=&BurbankSmall
     letterSpacing=0
     monospacedWidth=1
     size=22
     skewAmount=0
     typefaceFontName=BurbankRegular-Black
     outlineSettings {
      bApplyOutlineToDropShadows=false
      bMiteredCorners=false
      bSeparateFillAlpha=false
      outlineMaterial=null
      outlineSize=0
      outlineColor=<rgba 1 1 1 0>
     }
    }
    renderTransform {angle=0, scale=<1 1>, shear=<-12 0>, translation=<0 0>}
    shadowOffset=<0 0>
   }
   slot {class=VerticalBoxSlot, properties {padding=<ltrb 0 40 0 0>}}
  }
  Spacer_302 {
   widgetClassPath=&Spacer
   parent=HorizontalBox_0
   widget=Spacer_302
   slot {class=HorizontalBoxSlot, properties {size {sizeRule=Fill, value=0.25}}}
  }
  Image_BackgroundBottomBar {
   widgetClassPath=&Image
   bIsVariable=true
   parent=Overlay_Modal
   widget=Image_BackgroundBottomBar
   properties {brush {
     bIsDynamicallyLoaded=false
     drawAs=Image
     imageType=NoImage
     mirroring=NoMirror
     resourceName=None
     resourceObject=/FortUILibrary/Foundation/Art_ToolKit/AngledBoxes/M_UI_MoreInfo_BottomDividerLIne_Blue
     tiling=NoTile
     imageSize=<32 26>
     margin=<ltrb 0 0 0 0>
     outlineSettings {
      bUseBrushTransparency=false
      roundingType=HalfHeightRadius
      width=0
      color {colorUseRule=UseColor_Specified, specifiedColor=<rgba 0 0 0 0>}
      cornerRadii=<0 0 0 0>
     }
     tintColor {colorUseRule=UseColor_Specified, specifiedColor=<rgba 0.3263 0.3457 0.4149 1>}
     uVRegion {bIsValid=false, max=<0 0>, min=<0 0>}
    }}
   slot {class=OverlaySlot, properties {horizontalAlignment=HAlign_Fill, verticalAlignment=VAlign_Bottom}}
  }
  ButtonHeightSB {
   widgetClassPath=&SizeBox
   parent=Overlay_Modal
   widget=ButtonHeightSB
   properties {bOverride_MaxDesiredHeight=true, maxDesiredHeight=120}
   slot {class=OverlaySlot, properties {horizontalAlignment=HAlign_Fill, verticalAlignment=VAlign_Bottom, padding=<ltrb 0 32 0 -44>}}
  }
  HorizontalBox_149 {
   widgetClassPath=&HorizontalBox
   parent=ButtonHeightSB
   widget=HorizontalBox_149
   slot {class=SizeBoxSlot, properties {horizontalAlignment=HAlign_Center, verticalAlignment=VAlign_Top}}
  }
  Button_Cancel {
   widgetClassPath=/Game/Athena/UI/GenericModalButton
   bInherited=true
   bIsVariable=true
   parent=HorizontalBox_149
   widget=Button_Cancel
   properties {
    auto wrap text=false
    button text="No, Thanks"
    fontSize=24
    fontSizeOverride=true
    material Brush=/Game/Athena/UI/Frontend/Spatial/Art/M_Ui_Skewed_SpatialButton_MoreInfo_Blue
    minWidth=192
    onButtonBaseClicked=(&self:WidgetTree.Button_Cancel.HandleOnButtonClicked)
    onButtonBaseDoubleClicked=(&self:WidgetTree.Button_Cancel.HandleOnButtonDoubleClicked)
    onButtonBaseHovered=(&self:WidgetTree.Button_Cancel.HandleOnButtonHovered)
    onButtonBaseUnhovered=(&self:WidgetTree.Button_Cancel.HandleOnButtonUnhovered)
    onSelectedChangedBase=(&self:WidgetTree.Button_Cancel.HandleOnSelectedChanged)
    right Image=null
    shearText=true
    simpleHover=true
    hoveredSlateSoundOverride {resourceObject=/Game/Sounds/Fort_UI_Sounds/FrontEnd/DefaultSounds/PressedButtons/MSS_UI_Button_Hover_Default}
    pressedSlateSoundOverride {resourceObject=/Game/Sounds/Fort_UI_Sounds/FrontEnd/New_FrontEnd/Fortnite_UI_MainMenu_Back_01_Cue}
    triggeringInputAction {dataTable=&InputActionDataTable, rowName=GenericBack}
   }
   slot {class=HorizontalBoxSlot, properties {horizontalAlignment=HAlign_Center, padding=<ltrb 10 0 40 0>}}
  }
  Button_Confirm {
   widgetClassPath=&GenericModalButton
   bInherited=true
   bIsVariable=true
   parent=HorizontalBox_149
   widget=Button_Confirm
   properties {
    auto wrap text=false
    button text=Yes
    fontSize=24
    fontSizeOverride=true
    material Brush=&M_Ui_Skewed_SpatialButton_MoreInfo_Blue
    minWidth=192
    onButtonBaseClicked=(&self:WidgetTree.Button_Confirm.HandleOnButtonClicked)
    onButtonBaseDoubleClicked=(&self:WidgetTree.Button_Confirm.HandleOnButtonDoubleClicked)
    onButtonBaseHovered=(&self:WidgetTree.Button_Confirm.HandleOnButtonHovered)
    onButtonBaseUnhovered=(&self:WidgetTree.Button_Confirm.HandleOnButtonUnhovered)
    onSelectedChangedBase=(&self:WidgetTree.Button_Confirm.HandleOnSelectedChanged)
    right Image=null
    shearText=true
    simpleHover=true
    hoveredSlateSoundOverride {resourceObject=&MSS_UI_Button_Hover_Default}
    pressedSlateSoundOverride {resourceObject=&Fortnite_UI_MainMenu_Back_01_Cue}
    triggeringInputAction {dataTable=&InputActionDataTable, rowName=GenericForward}
   }
   slot {class=HorizontalBoxSlot, properties {horizontalAlignment=HAlign_Center}}
  }
  MobileClose {
   widgetClassPath=&SafeZone
   bIsVariable=true
   parent=Overlay
   widget=MobileClose
   slot {class=OverlaySlot, properties {horizontalAlignment=HAlign_Right, padding=<ltrb 0 32 32 0>}}
  }
  VisibilityBorder_Touch {
   widgetClassPath=&CommonHardwareVisibilityBorder
   parent=MobileClose
   widget=VisibilityBorder_Touch
   properties {visibilityQuery {
     autoDescription=" ANY( Input.Touch )"
     tokenStreamVersion=0
     userDescription=""
     queryTokenStream [0,1,1,1,0]
     tagDictionary [{tagName=Input.Touch}]
    }}
   slot {class=SafeZoneSlot}
  }
  Button_TouchClose {
   widgetClassPath=/FortUILibrary/Foundation/Buttons/CloseButton
   bInherited=true
   bIsVariable=true
   parent=VisibilityBorder_Touch
   widget=Button_TouchClose
   properties {
    onButtonBaseClicked=(&self:WidgetTree.Button_TouchClose.HandleOnButtonClicked)
    onButtonBaseDoubleClicked=(&self:WidgetTree.Button_TouchClose.HandleOnButtonDoubleClicked)
    onButtonBaseHovered=(&self:WidgetTree.Button_TouchClose.HandleOnButtonHovered)
    onButtonBaseUnhovered=(&self:WidgetTree.Button_TouchClose.HandleOnButtonUnhovered)
    onSelectedChangedBase=(&self:WidgetTree.Button_TouchClose.HandleOnSelectedChanged)
   }
   slot {class=BorderSlot, properties {padding=<ltrb 0 0 0 0>}}
  }
 ]
}
