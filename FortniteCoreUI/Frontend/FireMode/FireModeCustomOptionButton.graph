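# Widget Blueprint for a selectable custom option button on Fortnite's frontend fire mode screen, derived from the native FireModeCustomOptionButton class.
# It overlays off and on material images beneath a wrapped, uppercase option name label inherited from the parent.
# Selected and Deselected events each play two short animations, swapping image opacity and fading text between dark navy and white.
# The on image also animates material hover and IconScale parameters with an overshoot.
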WidgetBlueprint /FortniteCoreUI/Frontend/FireMode/FireModeCustomOptionButton
animations [
 Selected {
  binding_count=2
  length_seconds=0.2
  object=&self:Selected
  sequence {bindings [
    Baking_Unselected {
     id=#1
     possessed_class=Image
     tracks [{
       type=MovieSceneFloatTrack
       display_name=Render Opacity
       sections [{type=MovieSceneFloatSection, end_frame=4, start_frame=0, channels [None {type=MovieSceneScriptingFloatChannel, keys [[0,1],[1,0]]}]}]
      }]
    }
    Baking_Selected {
     id=#2
     possessed_class=Image
     tracks [
      {
       type=MovieSceneWidgetMaterialTrack
       sections [{
         type=MovieSceneParameterSection
         end_frame=4
         start_frame=0
         channels [
          hover {type=MovieSceneScriptingFloatChannel, keys [[0,0],[2.4,2],[4,1]]}
          IconScale {type=MovieSceneScriptingFloatChannel, keys [[0,0.1],[1.6,0.7],[3.2,0.5]]}
         ]
        }]
      }
      {
       type=MovieSceneFloatTrack
       display_name=Render Opacity
       sections [{type=MovieSceneFloatSection, end_frame=4, start_frame=0, channels [None {type=MovieSceneScriptingFloatChannel, keys [[0,0],[1,1]]}]}]
      }
     ]
    }
   ]}
 }
 TextSelected {
  binding_count=1
  length_seconds=0.2
  object=&self:TextSelected
  sequence {bindings [Text_CustomOptionName {
     id=#3
     display_name=FireModeCustomOptionText
     possessed_class=Image
     tracks [{
       type=MovieSceneColorTrack
       display_name=Color and Opacity
       sections [{
         type=MovieSceneColorSection
         end_frame=4
         start_frame=0
         channels [
          Color.R {type=MovieSceneScriptingFloatChannel, keys [[0,0],[4,1]]}
          Color.G {type=MovieSceneScriptingFloatChannel, keys [[0,0],[4,1]]}
          Color.B {type=MovieSceneScriptingFloatChannel, keys [[0,0.05],[4,1]]}
          Color.A {type=MovieSceneScriptingFloatChannel, const [1,2]}
         ]
        }]
      }]
    }]}
 }
 Un-Selected {
  binding_count=2
  length_seconds=0.2
  object=&self:Un-Selected
  sequence {bindings [
    Baking_Unselected {
     id=#1
     possessed_class=Image
     tracks [{
       type=MovieSceneFloatTrack
       display_name=Render Opacity
       sections [{type=MovieSceneFloatSection, end_frame=4, start_frame=0, channels [None {type=MovieSceneScriptingFloatChannel, keys [[3,0],[4,1]]}]}]
      }]
    }
    Baking_Selected {
     id=#2
     possessed_class=Image
     tracks [
      {
       type=MovieSceneWidgetMaterialTrack
       sections [{
         type=MovieSceneParameterSection
         end_frame=4
         start_frame=0
         channels [
          hover {type=MovieSceneScriptingFloatChannel, keys [[0,1],[1,2],[4,0]]}
          IconScale {type=MovieSceneScriptingFloatChannel, keys [[1,0.5],[3,0.7],[4,0.1]]}
         ]
        }]
      }
      {
       type=MovieSceneFloatTrack
       display_name=Render Opacity
       sections [{type=MovieSceneFloatSection, end_frame=4, start_frame=0, channels [None {type=MovieSceneScriptingFloatChannel, keys [[3,1],[4,0]]}]}]
      }
     ]
    }
   ]}
 }
 Un-TextSelected {
  binding_count=1
  length_seconds=0.2
  object=&self:Un-TextSelected
  sequence {bindings [Text_CustomOptionName {
     id=#3
     display_name=FireModeCustomOptionText
     possessed_class=Image
     tracks [{
       type=MovieSceneColorTrack
       display_name=Color and Opacity
       sections [{
         type=MovieSceneColorSection
         end_frame=4
         start_frame=0
         channels [
          Color.R {type=MovieSceneScriptingFloatChannel, keys [[0,1],[4,0]]}
          Color.G {type=MovieSceneScriptingFloatChannel, keys [[0,1],[4,0]]}
          Color.B {type=MovieSceneScriptingFloatChannel, keys [[0,1],[4,0.05]]}
          Color.A {type=MovieSceneScriptingFloatChannel, const [1,2]}
         ]
        }]
      }]
    }]}
 }
]
graphs [EventGraph {nodes [
   Event On Selected()
   > PlayAnimation(InAnimation=Get Selected() -> (Selected:Widget Animation Object Reference))
   > PlayAnimation(InAnimation=Get TextSelected() -> (TextSelected:Widget Animation Object Reference))
   Event On Deselected()
   > PlayAnimation(InAnimation=Get Un-Selected() -> (Un-Selected:Widget Animation Object Reference))
   > PlayAnimation(InAnimation=Get Un-TextSelected() -> (Un-TextSelected:Widget Animation Object Reference))
  ]}]
variables []
widgets {
 info {inheritedWidgetCount=2, parentClass=/Script/FortniteUI.FireModeCustomOptionButton, rootWidgetClass=&Overlay, widgetCount=8}
 widgets [
  Overlay_3 {widgetClassPath=&Overlay, widget=Overlay_3}
  BakingSwitcher {
   widgetClassPath=&Overlay
   parent=Overlay_3
   widget=BakingSwitcher
   slot {class=OverlaySlot, properties {horizontalAlignment=HAlign_Fill, verticalAlignment=VAlign_Fill}}
  }
  Baking_Unselected {
   widgetClassPath=&Image
   bIsVariable=true
   parent=BakingSwitcher
   widget=Baking_Unselected
   properties {brush {
     bIsDynamicallyLoaded=false
     drawAs=Image
     imageType=NoImage
     mirroring=NoMirror
     resourceName=None
     resourceObject=/FortniteCoreUI/Frontend/FireMode/art/MI_Bacchus_Firemode_OptionsOff
     tiling=NoTile
     imageSize=<32 32>
     margin=<ltrb 0.2 0.2 0.2 0.2>
     outlineSettings {
      bUseBrushTransparency=false
      roundingType=HalfHeightRadius
      width=0
      color {colorUseRule=UseColor_Specified, specifiedColor=<rgba 0 0 0 0>}
      cornerRadii=<0 0 0 0>
     }
     tintColor {colorUseRule=UseColor_Specified, specifiedColor=<rgba 1 1 1 1>}
     uVRegion {bIsValid=false, max=<0 0>, min=<0 0>}
    }}
   slot {class=OverlaySlot, properties {horizontalAlignment=HAlign_Fill, verticalAlignment=VAlign_Fill}}
  }
  Baking_Selected {
   widgetClassPath=&Image
   bIsVariable=true
   parent=BakingSwitcher
   widget=Baking_Selected
   properties {
    renderOpacity=0
    brush {
     bIsDynamicallyLoaded=false
     drawAs=Image
     imageType=NoImage
     mirroring=NoMirror
     resourceName=None
     resourceObject=/FortniteCoreUI/Frontend/FireMode/art/MI_Bacchus_Firemode_optionsOn
     tiling=NoTile
     imageSize=<32 32>
     margin=<ltrb 0 0 0 0>
     outlineSettings {
      bUseBrushTransparency=false
      roundingType=HalfHeightRadius
      width=0
      color {colorUseRule=UseColor_Specified, specifiedColor=<rgba 0 0 0 0>}
      cornerRadii=<0 0 0 0>
     }
     tintColor {colorUseRule=UseColor_Specified, specifiedColor=<rgba 1 1 1 1>}
     uVRegion {bIsValid=false, max=<0 0>, min=<0 0>}
    }
   }
   slot {class=OverlaySlot, properties {horizontalAlignment=HAlign_Fill, verticalAlignment=VAlign_Fill}}
  }
  ButtonText {
   widgetClassPath=&Overlay
   parent=Overlay_3
   widget=ButtonText
   slot {class=OverlaySlot, properties {horizontalAlignment=HAlign_Fill, verticalAlignment=VAlign_Fill, padding=<ltrb 40 0 10 0>}}
  }
  SizeBox_1 {
   widgetClassPath=&SizeBox
   parent=ButtonText
   widget=SizeBox_1
   properties {bOverride_HeightOverride=true, bOverride_WidthOverride=true, heightOverride=80, widthOverride=180}
   slot {class=OverlaySlot, properties {horizontalAlignment=HAlign_Center, verticalAlignment=VAlign_Center}}
  }
  Text_CustomOptionName {
   widgetClassPath=&CommonTextBlock
   bInherited=true
   bIsVariable=true
   parent=SizeBox_1
   widget=Text_CustomOptionName
   properties {
    autoWrapText=true
    justification=Center
    lineHeightPercentage=0.9
    text="[CUSTOM OPTION NAME]"
    textTransformPolicy=ToUpper
    wrappingPolicy=AllowPerCharacterWrapping
    colorAndOpacity {colorUseRule=UseColor_Specified, specifiedColor=<rgba 0 0 0.05 1>}
    font {
     bForceMonospaced=false
     bMaterialIsStencil=false
     fontMaterial=null
     fontObject=&Burbank
     letterSpacing=0
     monospacedWidth=1
     size=18
     skewAmount=0
     typefaceFontName=Burbank-PlayerPower
     outlineSettings {
      bApplyOutlineToDropShadows=false
      bMiteredCorners=false
      bSeparateFillAlpha=false
      outlineMaterial=null
      outlineSize=0
      outlineColor=<rgba 0 0 0 1>
     }
    }
    shadowOffset=<0 0>
   }
   slot {class=SizeBoxSlot, properties {verticalAlignment=VAlign_Center, padding=<ltrb 0 0 0 1>}}
  }
  InputActionWidget {widgetClassPath=&CommonActionWidget, bInherited=true, widget=null}
 ]
}
