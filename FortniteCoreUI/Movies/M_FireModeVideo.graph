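# M_FireModeVideo is an unlit, translucent UI-domain material, apparently for showing the Fire Mode video in the interface.
# Its graph holds a single texture sample of MT_FireModeVideo, which by its name is a media texture carrying the movie frames.
# The graph has no parameters or other processing nodes.
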Material /FortniteCoreUI/Movies/M_FireModeVideo
graph {
 properties {
  bUsedWithStaticMesh=false
  blendMode=BLEND_Translucent
  materialDomain=MD_UI
  pixelDepthOffsetMode=PDOM_Legacy
  shadingModel=MSM_Unlit
  usedShadingModels=Unlit
 }
 nodes [TextureSample() {texture=/Game/Movies/MT_FireModeVideo}]
}
