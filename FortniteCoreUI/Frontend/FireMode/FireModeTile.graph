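# Fire mode selection tile widget, built on the native FireModeTile class, with art over a video layer, title, description, tags and a two-button custom options drawer.
# PreConstruct updates text, sets the art brush and shows the Recommended tag from IsRecommended.
# Selection and deselection play highlight animations and open or close the drawer, whose toggle event drives its slide.
# Movie playback changes animate art material parameters to hide or show the still image.
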WidgetBlueprint /FortniteCoreUI/Frontend/FireMode/FireModeTile
animations [
 OpenCustom {
  binding_count=2
  length_seconds=0.25
  object=&self:OpenCustom
  sequence {bindings [
    Button_ToggleCustomOptions {
     id=#1
     display_name=Button_CustomOptions
     possessed_class=Overlay
     tracks [{
       type=MovieScene2DTransformTrack
       sections [{
         type=MovieScene2DTransformSection
         end_frame=5
         start_frame=0
         channels [
          Translation.X {type=MovieSceneScriptingFloatChannel, const [0,2]}
          Translation.Y {type=MovieSceneScriptingFloatChannel, keys [[0,104],[3,-20],[5,0]]}
          Angle {type=MovieSceneScriptingFloatChannel, const [0,1]}
          Scale.X {type=MovieSceneScriptingFloatChannel, const [1,1]}
          Scale.Y {type=MovieSceneScriptingFloatChannel, const [1,1]}
          Shear.X {type=MovieSceneScriptingFloatChannel, const [0,1]}
          Shear.Y {type=MovieSceneScriptingFloatChannel, const [0,1]}
         ]
        }]
      }]
    }
    bakingCustomOption {
     id=#2
     possessed_class=Image
     tracks [
      {
       type=MovieSceneColorTrack
       display_name=Color and Opacity
       sections [{
         type=MovieSceneColorSection
         end_frame=5
         start_frame=0
         channels [
          Color.R {type=MovieSceneScriptingFloatChannel, keys [[0,0],[1,1]]}
          Color.G {type=MovieSceneScriptingFloatChannel, keys [[0,0.3744],[1,1]]}
          Color.B {type=MovieSceneScriptingFloatChannel, keys [[0,0.6094],[1,1]]}
          Color.A {type=MovieSceneScriptingFloatChannel, const [1,2]}
         ]
        }]
      }
      {
       type=MovieSceneFloatTrack
       display_name=Render Opacity
       sections [{type=MovieSceneFloatSection, end_frame=5, start_frame=0, channels [None {type=MovieSceneScriptingFloatChannel, keys [[0,0],[1,1]]}]}]
      }
     ]
    }
   ]}
 }
 Selected_Recommended {
  binding_count=4
  length_seconds=0.2
  object=&self:Selected_Recommended
  sequence {bindings [
    Tile {
     id=#3
     possessed_class=Overlay
     tracks [{
       type=MovieScene2DTransformTrack
       sections [{
         type=MovieScene2DTransformSection
         end_frame=4
         start_frame=0
         channels [
          Translation.X {type=MovieSceneScriptingFloatChannel}
          Translation.Y {type=MovieSceneScriptingFloatChannel}
          Angle {type=MovieSceneScriptingFloatChannel}
          Scale.X {type=MovieSceneScriptingFloatChannel, keys [[0,1],[4.0003,1.02]]}
          Scale.Y {type=MovieSceneScriptingFloatChannel, keys [[0,1],[4.0003,1.02]]}
          Shear.X {type=MovieSceneScriptingFloatChannel}
          Shear.Y {type=MovieSceneScriptingFloatChannel}
         ]
        }]
      }]
    }
    ArtContainer {
     id=#4
     possessed_class=Image
     tracks [{
       type=MovieSceneWidgetMaterialTrack
       sections [{
         type=MovieSceneParameterSection
         end_frame=4
         start_frame=0
         channels [hover {type=MovieSceneScriptingFloatChannel, keys [[0,0],[2,5],[4,1]]}]
        }]
      }]
    }
    bakingCustomOption {
     id=#5
     possessed_class=Image
     tracks [{
       type=MovieSceneColorTrack
       display_name=Color and Opacity
       sections [{
         type=MovieSceneColorSection
         end_frame=4
         start_frame=0
         channels [
          Color.R {type=MovieSceneScriptingFloatChannel, keys [[0,0],[4.0003,1]]}
          Color.G {type=MovieSceneScriptingFloatChannel, keys [[0,0.3744],[4.0003,1]]}
          Color.B {type=MovieSceneScriptingFloatChannel, keys [[0,0.6094],[4.0003,1]]}
          Color.A {type=MovieSceneScriptingFloatChannel, const [1,2]}
         ]
        }]
      }]
    }
    selected {
     id=#6
     display_name=Tag_Selected
     possessed_class=Image
     tracks [
      {
       type=MovieScene2DTransformTrack
       sections [{
         type=MovieScene2DTransformSection
         end_frame=4
         start_frame=0
         channels [
          Translation.X {type=MovieSceneScriptingFloatChannel}
          Translation.Y {type=MovieSceneScriptingFloatChannel}
          Angle {type=MovieSceneScriptingFloatChannel}
          Scale.X {type=MovieSceneScriptingFloatChannel, thinned_from=5, keys [[0,1.2],[2,1.2],[3,0.95],[4,1]]}
          Scale.Y {type=MovieSceneScriptingFloatChannel, keys [[0,1.2],[2,1],[3,0.8],[4,1]]}
          Shear.X {type=MovieSceneScriptingFloatChannel}
          Shear.Y {type=MovieSceneScriptingFloatChannel}
         ]
        }]
      }
      {
       type=MovieSceneFloatTrack
       display_name=Render Opacity
       sections [
        {type=MovieSceneFloatSection, end_frame=4, start_frame=4, channels [None {type=MovieSceneScriptingFloatChannel, const [1,1]}]}
        {type=MovieSceneFloatSection, end_frame=4, start_frame=0, channels [None {type=MovieSceneScriptingFloatChannel, keys [[0,0],[1,0],[4,1]]}]}
       ]
      }
     ]
    }
   ]}
 }
 HideImageAndBg {
  binding_count=1
  length_seconds=0.05
  object=&self:HideImageAndBg
  sequence {bindings [ArtContainer {
     id=#7
     possessed_class=Image
     tracks [{
       type=MovieSceneWidgetMaterialTrack
       sections [{
         type=MovieSceneParameterSection
         end_frame=1
         start_frame=0
         channels [
          ImageageOpacity {type=MovieSceneScriptingFloatChannel, keys [[0,1],[1,0]]}
          OverallOpacity {type=MovieSceneScriptingFloatChannel, const [1,2]}
          Background_Opacity {type=MovieSceneScriptingFloatChannel, keys [[0,0],[0.9997,1]]}
         ]
        }]
      }]
    }]}
 }
 CloseCustom {
  binding_count=2
  length_seconds=0.25
  object=&self:CloseCustom
  sequence {bindings [
    Button_ToggleCustomOptions {
     id=#1
     display_name=Button_CustomOptions
     possessed_class=Overlay
     tracks [{
       type=MovieScene2DTransformTrack
       sections [{
         type=MovieScene2DTransformSection
         end_frame=5
         start_frame=0
         channels [
          Translation.X {type=MovieSceneScriptingFloatChannel, const [0,2]}
          Translation.Y {type=MovieSceneScriptingFloatChannel, keys [[0,0],[3,34.2863],[5.0003,104]]}
          Angle {type=MovieSceneScriptingFloatChannel, const [0,1]}
          Scale.X {type=MovieSceneScriptingFloatChannel, const [1,1]}
          Scale.Y {type=MovieSceneScriptingFloatChannel, const [1,1]}
          Shear.X {type=MovieSceneScriptingFloatChannel, const [0,1]}
          Shear.Y {type=MovieSceneScriptingFloatChannel, const [0,1]}
         ]
        }]
      }]
    }
    bakingCustomOption {
     id=#2
     possessed_class=Image
     tracks [
      {
       type=MovieSceneColorTrack
       display_name=Color and Opacity
       sections [{
         type=MovieSceneColorSection
         end_frame=5
         start_frame=0
         channels [
          Color.R {type=MovieSceneScriptingFloatChannel, keys [[4.7957,1],[5,0]]}
          Color.G {type=MovieSceneScriptingFloatChannel, keys [[4.7957,1],[5,0.3744]]}
          Color.B {type=MovieSceneScriptingFloatChannel, keys [[4.7957,1],[5,0.6094]]}
          Color.A {type=MovieSceneScriptingFloatChannel, const [1,2]}
         ]
        }]
      }
      {
       type=MovieSceneFloatTrack
       display_name=Render Opacity
       sections [{type=MovieSceneFloatSection, end_frame=5, start_frame=0, channels [None {type=MovieSceneScriptingFloatChannel, keys [[4.7957,1],[5,0]]}]}]
      }
     ]
    }
   ]}
 }
 Un-Selected_Recommended {
  binding_count=4
  length_seconds=0.2
  object=&self:Un-Selected_Recommended
  sequence {bindings [
    Tile {
     id=#3
     possessed_class=Overlay
     tracks [{
       type=MovieScene2DTransformTrack
       sections [{
         type=MovieScene2DTransformSection
         end_frame=4
         start_frame=0
         channels [
          Translation.X {type=MovieSceneScriptingFloatChannel}
          Translation.Y {type=MovieSceneScriptingFloatChannel}
          Angle {type=MovieSceneScriptingFloatChannel}
          Scale.X {type=MovieSceneScriptingFloatChannel, keys [[0,1.02],[4.0003,1]]}
          Scale.Y {type=MovieSceneScriptingFloatChannel, keys [[0,1.02],[4.0003,1]]}
          Shear.X {type=MovieSceneScriptingFloatChannel}
          Shear.Y {type=MovieSceneScriptingFloatChannel}
         ]
        }]
      }]
    }
    ArtContainer {
     id=#4
     possessed_class=Image
     tracks [{
       type=MovieSceneWidgetMaterialTrack
       sections [{
         type=MovieSceneParameterSection
         end_frame=4
         start_frame=0
         channels [hover {type=MovieSceneScriptingFloatChannel, keys [[0,1],[2,5],[4,0]]}]
        }]
      }]
    }
    bakingCustomOption {
     id=#5
     possessed_class=Image
     tracks [{
       type=MovieSceneColorTrack
       display_name=Color and Opacity
       sections [{
         type=MovieSceneColorSection
         end_frame=4
         start_frame=0
         channels [
          Color.R {type=MovieSceneScriptingFloatChannel, keys [[0,1],[4.0003,0]]}
          Color.G {type=MovieSceneScriptingFloatChannel, keys [[0,1],[4.0003,0.3744]]}
          Color.B {type=MovieSceneScriptingFloatChannel, keys [[0,1],[4.0003,0.6094]]}
          Color.A {type=MovieSceneScriptingFloatChannel, const [1,2]}
         ]
        }]
      }]
    }
    selected {
     id=#6
     display_name=Tag_Selected
     possessed_class=Image
     tracks [
      {
       type=MovieScene2DTransformTrack
       sections [{
         type=MovieScene2DTransformSection
         end_frame=4
         start_frame=0
         channels [
          Translation.X {type=MovieSceneScriptingFloatChannel}
          Translation.Y {type=MovieSceneScriptingFloatChannel}
          Angle {type=MovieSceneScriptingFloatChannel}
          Scale.X {type=MovieSceneScriptingFloatChannel, keys [[0,1],[1,0.95],[2,1],[3.9997,1.2]]}
          Scale.Y {type=MovieSceneScriptingFloatChannel, keys [[0,1],[1,0.8],[2,1]]}
          Shear.X {type=MovieSceneScriptingFloatChannel}
          Shear.Y {type=MovieSceneScriptingFloatChannel}
         ]
        }]
      }
      {
       type=MovieSceneFloatTrack
       display_name=Render Opacity
       sections [
        {type=MovieSceneFloatSection, end_frame=4, start_frame=4, channels [None {type=MovieSceneScriptingFloatChannel}]}
        {
         type=MovieSceneFloatSection
         end_frame=4
         start_frame=0
         channels [None {type=MovieSceneScriptingFloatChannel, keys [[0,1],[2.9597,1],[3,0],[4,0]]}]
        }
       ]
      }
     ]
    }
   ]}
 }
 ShowImageAndBg {
  binding_count=1
  length_seconds=0.0541
  object=&self:ShowImageAndBg
  sequence {bindings [ArtContainer {
     id=#7
     possessed_class=Image
     tracks [{
       type=MovieSceneWidgetMaterialTrack
       sections [{
         type=MovieSceneParameterSection
         end_frame=1
         start_frame=0
         channels [
          ImageageOpacity {type=MovieSceneScriptingFloatChannel, const [1,2]}
          OverallOpacity {type=MovieSceneScriptingFloatChannel, const [1,2]}
          Background_Opacity {type=MovieSceneScriptingFloatChannel, const [0,3]}
         ]
        }]
      }]
    }]}
 }
]
graphs [EventGraph {nodes [
   Event On Selected()
   > StopAnimation(InAnimation=Get Un-Selected_Recommended() -> (Un-Selected_Recommended:Widget Animation Object Reference))
   > PlayAnimation(InAnimation=Get Selected_Recommended() -> (Selected_Recommended:Widget Animation Object Reference))
   > SetIsCustomOptionsPanelOpen!(execute, self:Fire Mode Tile Object Reference, bIsOpen:bool=true) -> (then)
   n1 = Event OnCustomOptionsPanelToggled!() -> (OutputDelegate, then, bIsOpen:bool) {is=title K2Node_Event}
   > StopAnimation(InAnimation=Select(
     Option 0:Widget Animation Object Reference=Get OpenCustom() -> (OpenCustom:Widget Animation Object Reference)
     Option 1:Widget Animation Object Reference=Get CloseCustom() -> (CloseCustom:Widget Animation Object Reference)
     Index=@n1.bIsOpen
    ) -> (ReturnValue:Widget Animation Object Reference))
   > PlayAnimation(InAnimation=Select(Option 0=Get CloseCustom(), Option 1=Get OpenCustom(), Index=@n1.bIsOpen))
   n2 = Event OnMoviePlaybackChanged!() -> (OutputDelegate, then, bIsPlaying:bool) {is=title K2Node_Event}
   > PlayAnimation(InAnimation=Select(
     Option 0=Get ShowImageAndBg() -> (ShowImageAndBg:Widget Animation Object Reference)
     Option 1=Get HideImageAndBg() -> (HideImageAndBg:Widget Animation Object Reference)
     Index=@n2.bIsPlaying
    ))
   Event PreConstruct()
   > UpdateDisplayedText(self:Fire Mode Tile Object Reference)
   > SetBrush(self=Get ArtContainer() -> (ArtContainer:Image Object Reference), InBrush=Get FireModeArtImage() -> (FireModeArtImage:Slate Brush Structure))
   > SetVisibility(
    self=Get Tag_Recommended() -> (Tag_Recommended:Common Border Object Reference)
    InVisibility=Select(
     Option 0:ESlateVisibility Enum=Hidden
     Option 1:ESlateVisibility Enum=SelfHitTestInvisible
     Index=Get IsRecommended() -> (IsRecommended:bool)
    ) -> (ReturnValue:ESlateVisibility Enum)
   )
   Event On Deselected()
   > StopAnimation(InAnimation=Get Selected_Recommended())
   > PlayAnimation(InAnimation=Get Un-Selected_Recommended())
   > SetIsCustomOptionsPanelOpen(bIsOpen=false)
  ]}]
variables [
 FireModeArtImage {
  type=SlateBrush
  default {
   bIsDynamicallyLoaded=false
   drawAs=Image
   imageType=NoImage
   mirroring=NoMirror
   resourceName=None
   resourceObject=/FortniteCoreUI/Frontend/FireMode/art/MI_Bacchus_Firemode_Tile_Autofire
   tiling=NoTile
   imageSize=<32 32>
   margin=<ltrb 0 0 0 0>
   outlineSettings {
    bUseBrushTransparency=false
    roundingType=HalfHeightRadius
    width=0
    color {colorUseRule=UseColor_Specified, specifiedColor=<rgba 0 0 0 0>}
    cornerRadii=<0 0 0 1>
   }
   tintColor {colorUseRule=UseColor_Specified, specifiedColor=<rgba 1 1 1 1>}
   uVRegion {bIsValid=false, max=<0 0>, min=<0 0>}
  }
 }
 IsRecommended {type=boolean, default=true}
]
widgets {
 info {inheritedWidgetCount=8, parentClass=/Script/FortniteUI.FireModeTile, rootWidgetClass=&Overlay, widgetCount=36}
 widgets [
  Tile {widgetClassPath=&Overlay, widget=Tile, properties {clipping=ClipToBounds}}
  SizeBox_2 {
   widgetClassPath=&SizeBox
   parent=Tile
   widget=SizeBox_2
   properties {bOverride_HeightOverride=true, bOverride_WidthOverride=true, heightOverride=720, widthOverride=576}
   slot {class=OverlaySlot, properties {horizontalAlignment=HAlign_Fill, verticalAlignment=VAlign_Fill}}
  }
  Overlay_3 {widgetClassPath=&Overlay, parent=SizeBox_2, widget=Overlay_3, properties {clipping=ClipToBounds}, slot {class=SizeBoxSlot}}
  Image_Video {
   widgetClassPath=&Image
   parent=Overlay_3
   widget=Image_Video
   properties {brush {
     bIsDynamicallyLoaded=false
     drawAs=Image
     imageType=NoImage
     mirroring=NoMirror
     resourceName=None
     resourceObject=/FortniteCoreUI/Movies/M_FireModeVideo
     tiling=NoTile
     imageSize=<32 32>
     margin=<ltrb 0 0 0 0>
     outlineSettings {
      bUseBrushTransparency=false
      roundingType=HalfHeightRadius
      width=0
      color {colorUseRule=UseColor_Specified, specifiedColor=<rgba 0 0 0 0>}
      cornerRadii=<0 0 0 0>
     }
     tintColor {colorUseRule=UseColor_Specified, specifiedColor=<rgba 1 1 1 1>}
     uVRegion {bIsValid=false, max=<0 0>, min=<0 0>}
    }}
   slot {class=OverlaySlot, properties {horizontalAlignment=HAlign_Fill, verticalAlignment=VAlign_Fill}}
  }
  ArtContainer {
   widgetClassPath=&Image
   bIsVariable=true
   parent=Overlay_3
   widget=ArtContainer
   properties {brush {
     bIsDynamicallyLoaded=false
     drawAs=Image
     imageType=NoImage
     mirroring=NoMirror
     resourceName=None
     resourceObject=&MI_Bacchus_Firemode_Tile_Autofire
     tiling=NoTile
     imageSize=<32 32>
     margin=<ltrb 0 0 0 0>
     outlineSettings {
      bUseBrushTransparency=false
      roundingType=HalfHeightRadius
      width=0
      color {colorUseRule=UseColor_Specified, specifiedColor=<rgba 0 0 0 0>}
      cornerRadii=<0 0 0 0>
     }
     tintColor {colorUseRule=UseColor_Specified, specifiedColor=<rgba 1 1 1 1>}
     uVRegion {bIsValid=false, max=<0 0>, min=<0 0>}
    }}
   slot {class=OverlaySlot, properties {horizontalAlignment=HAlign_Fill, verticalAlignment=VAlign_Fill}}
  }
  Tag_Selected {
   widgetClassPath=&Border
   bIsVariable=true
   parent=Overlay_3
   widget=Tag_Selected
   properties {
    renderOpacity=0
    visibility=SelfHitTestInvisible
    background {
     bIsDynamicallyLoaded=false
     drawAs=Box
     imageType=NoImage
     mirroring=NoMirror
     resourceName=None
     resourceObject=/Game/Athena/UI/ReplayBrowser/M_Ui_Skewed_NoV1
     tiling=NoTile
     imageSize=<32 32>
     margin=<ltrb 0.25 0 0.25 0>
     outlineSettings {
      bUseBrushTransparency=false
      roundingType=HalfHeightRadius
      width=0
      color {colorUseRule=UseColor_Specified, specifiedColor=<rgba 0 0 0 0>}
      cornerRadii=<0 0 0 0>
     }
     tintColor {colorUseRule=UseColor_Specified, specifiedColor=<rgba 0.0343 1 0.227 1>}
     uVRegion {bIsValid=false, max=<0 0>, min=<0 0>}
    }
    contentColorAndOpacity=<rgba 0 0.147 0.063 1>
    padding=<ltrb 10 2 10 2>
   }
   slot {class=OverlaySlot, properties {horizontalAlignment=HAlign_Right, padding=<ltrb 20 20 20 20>}}
  }
  HorizontalBox_5 {
   widgetClassPath=&HorizontalBox
   parent=Tag_Selected
   widget=HorizontalBox_5
   slot {class=BorderSlot, properties {padding=<ltrb 10 2 10 2>}}
  }
  Image_140 {
   widgetClassPath=&Image
   bIsVariable=true
   parent=HorizontalBox_5
   widget=Image_140
   properties {
    brush {
     bIsDynamicallyLoaded=false
     drawAs=Image
     imageType=NoImage
     mirroring=NoMirror
     resourceName=None
     resourceObject=/Game/Athena/HUD/Bacchus/HUDLayoutTool/HLT_Icons/T-HLT-CheckMark-32
     tiling=NoTile
     imageSize=<16 16>
     margin=<ltrb 0 0 0 0>
     outlineSettings {
      bUseBrushTransparency=false
      roundingType=HalfHeightRadius
      width=0
      color {colorUseRule=UseColor_Specified, specifiedColor=<rgba 0 0 0 0>}
      cornerRadii=<0 0 0 0>
     }
     tintColor {colorUseRule=UseColor_Specified, specifiedColor=<rgba 1 1 1 1>}
     uVRegion {bIsValid=false, max=<0 0>, min=<0 0>}
    }
    renderTransform {angle=0, scale=<1.3 1.3>, shear=<0 0>, translation=<0 0>}
   }
   slot {class=HorizontalBoxSlot, properties {horizontalAlignment=HAlign_Left, verticalAlignment=VAlign_Center, padding=<ltrb 0 -1.5 6 0>}}
  }
  CommonTextBlock_10 {
   widgetClassPath=&CommonTextBlock
   parent=HorizontalBox_5
   widget=CommonTextBlock_10
   properties {
    text="Selected!"
    textTransformPolicy=ToUpper
    font {
     bForceMonospaced=false
     bMaterialIsStencil=false
     fontMaterial=null
     fontObject=&BurbankSmall
     letterSpacing=0
     monospacedWidth=1
     size=15
     skewAmount=0
     typefaceFontName=Black
     outlineSettings {
      bApplyOutlineToDropShadows=false
      bMiteredCorners=false
      bSeparateFillAlpha=false
      outlineMaterial=null
      outlineSize=0
      outlineColor=<rgba 0 0 0 1>
     }
    }
    renderTransform {angle=0, scale=<1 1>, shear=<-8 0>, translation=<0 0>}
    shadowOffset=<0 0>
   }
   slot {class=HorizontalBoxSlot}
  }
  Tag_Recommended {
   widgetClassPath=&CommonBorder
   bIsVariable=true
   parent=Overlay_3
   widget=Tag_Recommended
   properties {
    background {
     bIsDynamicallyLoaded=false
     drawAs=Image
     imageType=NoImage
     mirroring=NoMirror
     resourceName=None
     resourceObject=/FortniteCoreUI/Frontend/FireMode/art/MI_Bacchus_Firemode_Recommended
     tiling=NoTile
     imageSize=<32 32>
     margin=<ltrb 0 0 0 0>
     outlineSettings {
      bUseBrushTransparency=false
      roundingType=HalfHeightRadius
      width=0
      color {colorUseRule=UseColor_Specified, specifiedColor=<rgba 0 0 0 0>}
      cornerRadii=<0 0 0 0>
     }
     tintColor {colorUseRule=UseColor_Specified, specifiedColor=<rgba 1 1 1 1>}
     uVRegion {bIsValid=false, max=<0 0>, min=<0 0>}
    }
    padding=<ltrb 0 0 15 0>
   }
   slot {class=OverlaySlot}
  }
  SizeBox_0 {
   widgetClassPath=&SizeBox
   parent=Tag_Recommended
   widget=SizeBox_0
   properties {
    bOverride_HeightOverride=true
    bOverride_MaxDesiredWidth=true
    heightOverride=40
    maxDesiredWidth=230
    renderTransform {angle=0, scale=<1 1>, shear=<-7 0>, translation=<0 0>}
   }
   slot {class=BorderSlot, properties {padding=<ltrb 0 0 15 0>}}
  }
  ScaleBox_1 {
   widgetClassPath=&ScaleBox
   parent=SizeBox_0
   widget=ScaleBox_1
   properties {stretch=ScaleToFitX, stretchDirection=DownOnly}
   slot {class=SizeBoxSlot, properties {verticalAlignment=VAlign_Center, padding=<ltrb 10 0 10 -1>}}
  }
  Text_RecommendedTag {
   widgetClassPath=&CommonTextBlock
   bInherited=true
   bIsVariable=true
   parent=ScaleBox_1
   widget=Text_RecommendedTag
   properties {
    text="[Recommended]"
    textTransformPolicy=ToUpper
    colorAndOpacity {colorUseRule=UseColor_Specified, specifiedColor=<rgba 0 0 0 1>}
    font {
     bForceMonospaced=false
     bMaterialIsStencil=false
     fontMaterial=null
     fontObject=&BurbankSmall
     letterSpacing=0
     monospacedWidth=1
     size=16
     skewAmount=0
     typefaceFontName=Black
     outlineSettings {
      bApplyOutlineToDropShadows=false
      bMiteredCorners=false
      bSeparateFillAlpha=false
      outlineMaterial=null
      outlineSize=0
      outlineColor=<rgba 0 0 0 1>
     }
    }
    shadowOffset=<0 0>
   }
   slot {class=ScaleBoxSlot, properties {horizontalAlignment=HAlign_Fill}}
  }
  headerTitle {
   widgetClassPath=&Overlay
   parent=Overlay_3
   widget=headerTitle
   slot {class=OverlaySlot, properties {horizontalAlignment=HAlign_Fill}}
  }
  VerticalBox_1 {
   widgetClassPath=&VerticalBox
   parent=headerTitle
   widget=VerticalBox_1
   slot {class=OverlaySlot, properties {horizontalAlignment=HAlign_Fill, verticalAlignment=VAlign_Center, padding=<ltrb 25 37 25 0>}}
  }
  ScaleBox_0 {
   widgetClassPath=&ScaleBox
   parent=VerticalBox_1
   widget=ScaleBox_0
   properties {stretch=ScaleToFitX, stretchDirection=DownOnly}
   slot {class=VerticalBoxSlot}
  }
  Text_Title {
   widgetClassPath=&CommonTextBlock
   bInherited=true
   bIsVariable=true
   parent=ScaleBox_0
   widget=Text_Title
   properties {
    justification=Center
    text="[TITLE]"
    textTransformPolicy=ToUpper
    font {
     bForceMonospaced=false
     bMaterialIsStencil=false
     fontMaterial=null
     fontObject=&Burbank
     letterSpacing=0
     monospacedWidth=1
     size=45
     skewAmount=0
     typefaceFontName=Burbank-PlayerPower
     outlineSettings {
      bApplyOutlineToDropShadows=false
      bMiteredCorners=false
      bSeparateFillAlpha=false
      outlineMaterial=null
      outlineSize=0
      outlineColor=<rgba 0 0 0 1>
     }
    }
    shadowOffset=<0 0>
   }
   slot {class=ScaleBoxSlot, properties {horizontalAlignment=HAlign_Fill}}
  }
  SizeBox_1 {
   widgetClassPath=&SizeBox
   parent=VerticalBox_1
   widget=SizeBox_1
   properties {bOverride_WidthOverride=true, widthOverride=390}
   slot {class=VerticalBoxSlot, properties {horizontalAlignment=HAlign_Center, padding=<ltrb 0 10 0 0>}}
  }
  Text_Description {
   widgetClassPath=&CommonTextBlock
   bInherited=true
   bIsVariable=true
   parent=SizeBox_1
   widget=Text_Description
   properties {
    autoWrapText=true
    justification=Center
    text="[Description]"
    font {
     bForceMonospaced=false
     bMaterialIsStencil=false
     fontMaterial=null
     fontObject=&BurbankSmall
     letterSpacing=0
     monospacedWidth=1
     size=22
     skewAmount=0
     typefaceFontName=Bold
     outlineSettings {
      bApplyOutlineToDropShadows=false
      bMiteredCorners=false
      bSeparateFillAlpha=false
      outlineMaterial=null
      outlineSize=0
      outlineColor=<rgba 0 0 0 1>
     }
    }
    shadowOffset=<0 0>
   }
   slot {class=SizeBoxSlot}
  }
  Button_ToggleCustomOptions {
   widgetClassPath=&Button
   bInherited=true
   parent=Overlay_3
   widget=Button_ToggleCustomOptions
   properties {
    renderTransform {angle=0, scale=<1 1>, shear=<0 0>, translation=<0 104>}
    widgetStyle {
     clickedSlateSound {resourceObject=null}
     disabled {
      bIsDynamicallyLoaded=false
      drawAs=NoDrawType
      imageType=NoImage
      mirroring=NoMirror
      resourceName=None
      resourceObject=null
      tiling=NoTile
      imageSize=<0 0>
      margin=<ltrb 0 0 0 0>
      outlineSettings {
       bUseBrushTransparency=false
       roundingType=HalfHeightRadius
       width=0
       color {colorUseRule=UseColor_Specified, specifiedColor=<rgba 0 0 0 0>}
       cornerRadii=<0 0 0 0>
      }
      tintColor {colorUseRule=UseColor_Specified, specifiedColor=<rgba 1 1 1 1>}
      uVRegion {bIsValid=false, max=<0 0>, min=<0 0>}
     }
     disabledForeground {colorUseRule=UseColor_Specified, specifiedColor=<rgba 0.5271 0.5271 0.5271 1>}
     hovered {
      bIsDynamicallyLoaded=false
      drawAs=RoundedBox
      imageType=NoImage
      mirroring=NoMirror
      resourceName=None
      resourceObject=null
      tiling=NoTile
      imageSize=<32 32>
      margin=<ltrb 0 0 0 0>
      outlineSettings {
       bUseBrushTransparency=true
       roundingType=FixedRadius
       width=1
       color {colorUseRule=UseColor_Specified, specifiedColor=<rgba 0.7243 0.7243 0.7243 1>}
       cornerRadii=<4 4 4 4>
      }
      tintColor {colorUseRule=UseColor_Specified, specifiedColor=<rgba 0.7243 0.7243 0.7243 0>}
      uVRegion {bIsValid=false, max=<0 0>, min=<0 0>}
     }
     hoveredForeground {colorUseRule=UseColor_Specified, specifiedColor=<rgba 0 1 1 1>}
     hoveredSlateSound {resourceObject=null}
     normal {
      bIsDynamicallyLoaded=false
      drawAs=RoundedBox
      imageType=NoImage
      mirroring=NoMirror
      resourceName=None
      resourceObject=null
      tiling=NoTile
      imageSize=<32 32>
      margin=<ltrb 0 0 0 0>
      outlineSettings {
       bUseBrushTransparency=true
       roundingType=FixedRadius
       width=1
       color {colorUseRule=UseColor_Specified, specifiedColor=<rgba 0.6951 0.6951 0.6951 1>}
       cornerRadii=<4 4 4 4>
      }
      tintColor {colorUseRule=UseColor_Specified, specifiedColor=<rgba 0.4955 0.4955 0.4955 0>}
      uVRegion {bIsValid=false, max=<0 0>, min=<0 0>}
     }
     normalForeground {colorUseRule=UseColor_Specified, specifiedColor=<rgba 0 1 1 1>}
     normalPadding=<ltrb 12 1.5 12 1.5>
     pressed {
      bIsDynamicallyLoaded=false
      drawAs=RoundedBox
      imageType=NoImage
      mirroring=NoMirror
      resourceName=None
      resourceObject=null
      tiling=NoTile
      imageSize=<32 32>
      margin=<ltrb 0 0 0 0>
      outlineSettings {
       bUseBrushTransparency=true
       roundingType=FixedRadius
       width=1
       color {colorUseRule=UseColor_Specified, specifiedColor=<rgba 0.7243 0.7243 0.7243 1>}
       cornerRadii=<4 4 4 4>
      }
      tintColor {colorUseRule=UseColor_Specified, specifiedColor=<rgba 0.7243 0.7243 0.7243 0>}
      uVRegion {bIsValid=false, max=<0 0>, min=<0 0>}
     }
     pressedForeground {colorUseRule=UseColor_Specified, specifiedColor=<rgba 0 1 1 1>}
     pressedPadding=<ltrb 12 1.5 12 1.5>
     pressedSlateSound {resourceObject=null}
    }
   }
   slot {class=OverlaySlot, properties {horizontalAlignment=HAlign_Fill, verticalAlignment=VAlign_Bottom}}
  }
  VerticalBox_0 {
   widgetClassPath=&VerticalBox
   parent=Button_ToggleCustomOptions
   widget=VerticalBox_0
   slot {class=ButtonSlot, properties {horizontalAlignment=HAlign_Fill, verticalAlignment=VAlign_Bottom, padding=<ltrb 0 0 0 0>}}
  }
  Tab {
   widgetClassPath=&Overlay
   parent=VerticalBox_0
   widget=Tab
   slot {class=VerticalBoxSlot, properties {horizontalAlignment=HAlign_Center, verticalAlignment=VAlign_Bottom}}
  }
  bakingCustomOptionsButton {
   widgetClassPath=&Image
   bIsVariable=true
   parent=Tab
   widget=bakingCustomOptionsButton
   properties {brush {
     bIsDynamicallyLoaded=false
     drawAs=Image
     imageType=NoImage
     mirroring=NoMirror
     resourceName=None
     resourceObject=/FortniteCoreUI/Frontend/FireMode/art/MI_BacchusFiremode_CustomTabTitle
     tiling=NoTile
     imageSize=<64 32>
     margin=<ltrb 0.5 0 0 0>
     outlineSettings {
      bUseBrushTransparency=false
      roundingType=HalfHeightRadius
      width=0
      color {colorUseRule=UseColor_Specified, specifiedColor=<rgba 0 0 0 0>}
      cornerRadii=<0 0 0 0>
     }
     tintColor {colorUseRule=UseColor_Specified, specifiedColor=<rgba 1 1 1 1>}
     uVRegion {bIsValid=false, max=<0 0>, min=<0 0>}
    }}
   slot {class=OverlaySlot, properties {horizontalAlignment=HAlign_Fill, verticalAlignment=VAlign_Fill}}
  }
  HorizontalBox_3 {
   widgetClassPath=&HorizontalBox
   parent=Tab
   widget=HorizontalBox_3
   slot {class=OverlaySlot, properties {horizontalAlignment=HAlign_Fill, verticalAlignment=VAlign_Bottom, padding=<ltrb 20 20 20 20>}}
  }
  Image_75 {
   widgetClassPath=&Image
   bIsVariable=true
   parent=HorizontalBox_3
   widget=Image_75
   properties {brush {
     bIsDynamicallyLoaded=false
     drawAs=Image
     imageType=NoImage
     mirroring=NoMirror
     resourceName=None
     resourceObject=/Game/Athena/HUD/Art/BacchusHudIcons/QuickHeal_Normal
     tiling=NoTile
     imageSize=<26 26>
     margin=<ltrb 0 0 0 0>
     outlineSettings {
      bUseBrushTransparency=false
      roundingType=HalfHeightRadius
      width=0
      color {colorUseRule=UseColor_Specified, specifiedColor=<rgba 0 0 0 0>}
      cornerRadii=<0 0 0 0>
     }
     tintColor {colorUseRule=UseColor_Specified, specifiedColor=<rgba 1 1 1 1>}
     uVRegion {bIsValid=false, max=<0 0>, min=<0 0>}
    }}
   slot {class=HorizontalBoxSlot, properties {horizontalAlignment=HAlign_Center, verticalAlignment=VAlign_Center}}
  }
  Text_CustomOptionsTabName {
   widgetClassPath=&CommonTextBlock
   bInherited=true
   bIsVariable=true
   parent=HorizontalBox_3
   widget=Text_CustomOptionsTabName
   properties {
    text="[Custom Options]"
    textTransformPolicy=ToUpper
    wrapTextAt=400
    font {
     bForceMonospaced=false
     bMaterialIsStencil=false
     fontMaterial=null
     fontObject=&BurbankSmall
     letterSpacing=0
     monospacedWidth=1
     size=16
     skewAmount=0
     typefaceFontName=Black
     outlineSettings {
      bApplyOutlineToDropShadows=false
      bMiteredCorners=false
      bSeparateFillAlpha=false
      outlineMaterial=null
      outlineSize=0
      outlineColor=<rgba 0 0 0 1>
     }
    }
    renderTransform {angle=0, scale=<1 1>, shear=<-7 0>, translation=<0 0>}
    shadowOffset=<0 0>
   }
   slot {class=HorizontalBoxSlot, properties {padding=<ltrb 12 0 0 0>}}
  }
  ContainerCustomOptions {widgetClassPath=&Overlay, parent=VerticalBox_0, widget=ContainerCustomOptions, slot {class=VerticalBoxSlot}}
  bakingCustomOption {
   widgetClassPath=&Image
   bIsVariable=true
   parent=ContainerCustomOptions
   widget=bakingCustomOption
   properties {colorAndOpacity=<rgba 0 0.3744 0.6094 1>}
   slot {class=OverlaySlot, properties {horizontalAlignment=HAlign_Fill, verticalAlignment=VAlign_Fill, padding=<ltrb -5 0 -5 -30>}}
  }
  ButtonsHeigh {
   widgetClassPath=&SizeBox
   parent=ContainerCustomOptions
   widget=ButtonsHeigh
   properties {bOverride_HeightOverride=true, heightOverride=90}
   slot {class=OverlaySlot, properties {horizontalAlignment=HAlign_Fill, verticalAlignment=VAlign_Fill, padding=<ltrb 0 10 0 10>}}
  }
  HB_Buttons {widgetClassPath=&HorizontalBox, parent=ButtonsHeigh, widget=HB_Buttons, slot {class=SizeBoxSlot}}
  SizeBox_3 {
   widgetClassPath=&SizeBox
   parent=HB_Buttons
   widget=SizeBox_3
   properties {bOverride_WidthOverride=true, widthOverride=270}
   slot {class=HorizontalBoxSlot, properties {horizontalAlignment=HAlign_Center, size {sizeRule=Fill, value=1}}}
  }
  Button_CustomOption1 {
   widgetClassPath=&/FireModeCustomOptionButton
   bInherited=true
   bIsVariable=true
   parent=SizeBox_3
   widget=Button_CustomOption1
   slot {class=SizeBoxSlot}
  }
  Spacer_62 {widgetClassPath=&Spacer, parent=HB_Buttons, widget=Spacer_62, properties {size=<10 1>}, slot {class=HorizontalBoxSlot}}
  SizeBox_4 {
   widgetClassPath=&SizeBox
   parent=HB_Buttons
   widget=SizeBox_4
   properties {bOverride_WidthOverride=true, widthOverride=270}
   slot {class=HorizontalBoxSlot, properties {horizontalAlignment=HAlign_Right, padding=<ltrb 0 0 -3 0>, size {sizeRule=Fill, value=1}}}
  }
  Button_CustomOption2 {
   widgetClassPath=&FireModeCustomOptionButton
   bInherited=true
   bIsVariable=true
   parent=SizeBox_4
   widget=Button_CustomOption2
   slot {class=SizeBoxSlot}
  }
  InputActionWidget {widgetClassPath=&CommonActionWidget, bInherited=true, widget=null}
 ]
}
