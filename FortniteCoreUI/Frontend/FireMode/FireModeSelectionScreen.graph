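# Frontend screen, built on the native FireModeSelectionScreen class, where the player picks how weapons fire.
# Three FireModeTile widgets offer tap anywhere, auto-fire and dedicated button presets, each with art, video ID and custom options.
# A footer holds a Settings reminder and Confirm button, and a close button appears only for touch input.
# The graph only plays slide-and-fade In/Out animations on activation and deactivation.
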WidgetBlueprint /FortniteCoreUI/Frontend/FireMode/FireModeSelectionScreen
animations [
 In {
  binding_count=3
  length_seconds=0.25
  object=&self:In
  sequence {bindings [
    HorizontalBox_3 {
     id=#1
     possessed_class=HorizontalBox
     tracks [
      {
       type=MovieSceneFloatTrack
       display_name=Render Opacity
       sections [{type=MovieSceneFloatSection, end_frame=8, start_frame=3, channels [None {type=MovieSceneScriptingFloatChannel, keys [[7,0],[8.0007,1]]}]}]
      }
      {
       type=MovieScene2DTransformTrack
       sections [{
         type=MovieScene2DTransformSection
         end_frame=8
         start_frame=3
         channels [
          Translation.X {type=MovieSceneScriptingFloatChannel, keys [[7,-300],[8.0007,0]]}
          Translation.Y {type=MovieSceneScriptingFloatChannel}
          Angle {type=MovieSceneScriptingFloatChannel}
          Scale.X {type=MovieSceneScriptingFloatChannel}
          Scale.Y {type=MovieSceneScriptingFloatChannel}
          Shear.X {type=MovieSceneScriptingFloatChannel}
          Shear.Y {type=MovieSceneScriptingFloatChannel}
         ]
        }]
      }
     ]
    }
    FireModeTileRow {
     id=#2
     possessed_class=HorizontalBox
     tracks [
      {
       type=MovieSceneFloatTrack
       display_name=Render Opacity
       sections [{type=MovieSceneFloatSection, end_frame=7, start_frame=3, channels [None {type=MovieSceneScriptingFloatChannel, keys [[3,0],[7.9947,1]]}]}]
      }
      {
       type=MovieScene2DTransformTrack
       sections [{
         type=MovieScene2DTransformSection
         end_frame=8
         start_frame=3
         channels [
          Translation.X {type=MovieSceneScriptingFloatChannel, keys [[4,-150],[7,0]]}
          Translation.Y {type=MovieSceneScriptingFloatChannel, const [0,1]}
          Angle {type=MovieSceneScriptingFloatChannel, const [0,1]}
          Scale.X {type=MovieSceneScriptingFloatChannel, const [1,1]}
          Scale.Y {type=MovieSceneScriptingFloatChannel, const [1,1]}
          Shear.X {type=MovieSceneScriptingFloatChannel, const [0,1]}
          Shear.Y {type=MovieSceneScriptingFloatChannel, const [0,1]}
         ]
        }]
      }
     ]
    }
    TitleRow {
     id=#3
     possessed_class=Overlay
     tracks [
      {
       type=MovieSceneFloatTrack
       display_name=Render Opacity
       sections [{type=MovieSceneFloatSection, end_frame=8, start_frame=3, channels [None {type=MovieSceneScriptingFloatChannel, keys [[3,0],[8.0007,1]]}]}]
      }
      {
       type=MovieScene2DTransformTrack
       sections [{
         type=MovieScene2DTransformSection
         end_frame=8
         start_frame=3
         channels [
          Translation.X {type=MovieSceneScriptingFloatChannel, keys [[3,-200],[8.0007,0]]}
          Translation.Y {type=MovieSceneScriptingFloatChannel}
          Angle {type=MovieSceneScriptingFloatChannel}
          Scale.X {type=MovieSceneScriptingFloatChannel}
          Scale.Y {type=MovieSceneScriptingFloatChannel}
          Shear.X {type=MovieSceneScriptingFloatChannel}
          Shear.Y {type=MovieSceneScriptingFloatChannel}
         ]
        }]
      }
     ]
    }
   ]}
 }
 Out {
  binding_count=3
  length_seconds=0.25
  object=&self:Out
  sequence {bindings [
    HorizontalBox_3 {
     id=#1
     possessed_class=HorizontalBox
     tracks [
      {
       type=MovieSceneFloatTrack
       display_name=Render Opacity
       sections [{type=MovieSceneFloatSection, end_frame=8, start_frame=3, channels [None {type=MovieSceneScriptingFloatChannel, keys [[7,1],[8.0007,0]]}]}]
      }
      {
       type=MovieScene2DTransformTrack
       sections [{
         type=MovieScene2DTransformSection
         end_frame=8
         start_frame=3
         channels [
          Translation.X {type=MovieSceneScriptingFloatChannel, keys [[7,0],[8.0007,-300]]}
          Translation.Y {type=MovieSceneScriptingFloatChannel}
          Angle {type=MovieSceneScriptingFloatChannel}
          Scale.X {type=MovieSceneScriptingFloatChannel}
          Scale.Y {type=MovieSceneScriptingFloatChannel}
          Shear.X {type=MovieSceneScriptingFloatChannel}
          Shear.Y {type=MovieSceneScriptingFloatChannel}
         ]
        }]
      }
     ]
    }
    FireModeTileRow {
     id=#2
     possessed_class=HorizontalBox
     tracks [
      {
       type=MovieSceneFloatTrack
       display_name=Render Opacity
       sections [{type=MovieSceneFloatSection, end_frame=7, start_frame=3, channels [None {type=MovieSceneScriptingFloatChannel, keys [[3,1],[7.9947,0]]}]}]
      }
      {
       type=MovieScene2DTransformTrack
       sections [{
         type=MovieScene2DTransformSection
         end_frame=8
         start_frame=3
         channels [
          Translation.X {type=MovieSceneScriptingFloatChannel, keys [[4,0],[7,-150]]}
          Translation.Y {type=MovieSceneScriptingFloatChannel, const [0,1]}
          Angle {type=MovieSceneScriptingFloatChannel, const [0,1]}
          Scale.X {type=MovieSceneScriptingFloatChannel, const [1,1]}
          Scale.Y {type=MovieSceneScriptingFloatChannel, const [1,1]}
          Shear.X {type=MovieSceneScriptingFloatChannel, const [0,1]}
          Shear.Y {type=MovieSceneScriptingFloatChannel, const [0,1]}
         ]
        }]
      }
     ]
    }
    TitleRow {
     id=#3
     possessed_class=Overlay
     tracks [
      {
       type=MovieSceneFloatTrack
       display_name=Render Opacity
       sections [{type=MovieSceneFloatSection, end_frame=8, start_frame=3, channels [None {type=MovieSceneScriptingFloatChannel, keys [[3,1],[8.0007,0]]}]}]
      }
      {
       type=MovieScene2DTransformTrack
       sections [{
         type=MovieScene2DTransformSection
         end_frame=8
         start_frame=3
         channels [
          Translation.X {type=MovieSceneScriptingFloatChannel, keys [[3,0],[8.0007,-200]]}
          Translation.Y {type=MovieSceneScriptingFloatChannel}
          Angle {type=MovieSceneScriptingFloatChannel}
          Scale.X {type=MovieSceneScriptingFloatChannel}
          Scale.Y {type=MovieSceneScriptingFloatChannel}
          Shear.X {type=MovieSceneScriptingFloatChannel}
          Shear.Y {type=MovieSceneScriptingFloatChannel}
         ]
        }]
      }
     ]
    }
   ]}
 }
]
graphs [EventGraph {nodes [
   Event On Activated()
   > PlayAnimation(InAnimation=Get In() -> (In:Widget Animation Object Reference))
   Event On Deactivated()
   > PlayAnimation(InAnimation=Get Out() -> (Out:Widget Animation Object Reference))
  ]}]
variables []
widgets {
 info {inheritedWidgetCount=4, parentClass=/Script/FortniteUI.FireModeSelectionScreen, rootWidgetClass=&Overlay, widgetCount=26}
 widgets [
  Overlay_29 {widgetClassPath=&Overlay, widget=Overlay_29}
  BackgroundGradient {
   widgetClassPath=&Image
   parent=Overlay_29
   widget=BackgroundGradient
   properties {
    visibility=SelfHitTestInvisible
    brush {
     bIsDynamicallyLoaded=false
     drawAs=Image
     imageType=NoImage
     mirroring=NoMirror
     resourceName=None
     resourceObject=/Game/Athena/UI/Frontend/Art/M_UI_RadialGradientMaster_Dark2
     tiling=NoTile
     imageSize=<2160 2160>
     margin=<ltrb 0 0 0 0>
     outlineSettings {
      bUseBrushTransparency=false
      roundingType=HalfHeightRadius
      width=0
      color {colorUseRule=UseColor_Specified, specifiedColor=<rgba 0 0 0 0>}
      cornerRadii=<0 0 0 0>
     }
     tintColor {colorUseRule=UseColor_Specified, specifiedColor=<rgba 1 1 1 1>}
     uVRegion {bIsValid=false, max=<0 0>, min=<0 0>}
    }
   }
   slot {class=OverlaySlot, properties {horizontalAlignment=HAlign_Fill, verticalAlignment=VAlign_Fill}}
  }
  Overlay_1 {
   widgetClassPath=&Overlay
   parent=Overlay_29
   widget=Overlay_1
   slot {class=OverlaySlot, properties {horizontalAlignment=HAlign_Fill, verticalAlignment=VAlign_Fill}}
  }
  VerticalBox_0 {
   widgetClassPath=&VerticalBox
   parent=Overlay_1
   widget=VerticalBox_0
   slot {class=OverlaySlot, properties {horizontalAlignment=HAlign_Fill, verticalAlignment=VAlign_Fill}}
  }
  contents {
   widgetClassPath=&VerticalBox
   parent=VerticalBox_0
   widget=contents
   slot {class=VerticalBoxSlot, properties {verticalAlignment=VAlign_Center, size {sizeRule=Fill, value=1}}}
  }
  TitleRow {widgetClassPath=&Overlay, parent=contents, widget=TitleRow, slot {class=VerticalBoxSlot, properties {padding=<ltrb 0 60 0 60>}}}
  Text_Title {
   widgetClassPath=&CommonTextBlock
   bIsVariable=true
   parent=TitleRow
   widget=Text_Title
   properties {
    justification=Center
    text=How do you want to fire your weapon?
    textTransformPolicy=ToUpper
    font {
     bForceMonospaced=false
     bMaterialIsStencil=false
     fontMaterial=null
     fontObject=&Burbank
     letterSpacing=0
     monospacedWidth=1
     size=40
     skewAmount=0
     typefaceFontName=BurbankRegular-Black
     outlineSettings {
      bApplyOutlineToDropShadows=false
      bMiteredCorners=false
      bSeparateFillAlpha=false
      outlineMaterial=null
      outlineSize=0
      outlineColor=<rgba 0 0 0 1>
     }
    }
    renderTransform {angle=0, scale=<1 1>, shear=<-12 0>, translation=<0 0>}
    shadowOffset=<0 0>
   }
   slot {class=OverlaySlot, properties {horizontalAlignment=HAlign_Center, verticalAlignment=VAlign_Center}}
  }
  FireModeTileRow {
   widgetClassPath=&HorizontalBox
   bInherited=true
   bIsVariable=true
   parent=contents
   widget=FireModeTileRow
   slot {class=VerticalBoxSlot, properties {horizontalAlignment=HAlign_Center}}
  }
  DynamicLeftPadding {
   widgetClassPath=&Spacer
   parent=FireModeTileRow
   widget=DynamicLeftPadding
   slot {class=HorizontalBoxSlot, properties {size {sizeRule=Automatic, value=0.05}}}
  }
  FireModeTile_TapAnywhere {
   widgetClassPath=&/FireModeTile
   bIsVariable=true
   parent=FireModeTileRow
   widget=FireModeTile_TapAnywhere
   properties {
    customOption1=AutoFire
    customOption2=FireButton
    customOptionsTabName=Add Custom Options
    fireModeDescription=Fire by tapping anywhere on the screen
    isRecommended=false
    presetFireModeType=TapToShoot
    recommendedTag=Recommended
    videoUId=DPYdIEbtCUxWdjFrvm
    fireModeArtImage {
     bIsDynamicallyLoaded=false
     drawAs=Image
     imageType=NoImage
     mirroring=NoMirror
     resourceName=None
     resourceObject=/FortniteCoreUI/Frontend/FireMode/art/MI_Bacchus_Firemode_Tile_TapAnywhere
     tiling=NoTile
     imageSize=<576 720>
     margin=<ltrb 0 0 0 0>
     outlineSettings {
      bUseBrushTransparency=false
      roundingType=HalfHeightRadius
      width=0
      color {colorUseRule=UseColor_Specified, specifiedColor=<rgba 0 0 0 0>}
      cornerRadii=<0 0 0 1>
     }
     tintColor {colorUseRule=UseColor_Specified, specifiedColor=<rgba 1 1 1 1>}
     uVRegion {bIsValid=false, max=<0 0>, min=<0 0>}
    }
   }
   slot {
    class=HorizontalBoxSlot
    properties {horizontalAlignment=HAlign_Center, verticalAlignment=VAlign_Center, size {sizeRule=Automatic, value=0.33}}
   }
  }
  FireModeTile_AutoFire {
   widgetClassPath=&FireModeTile
   bIsVariable=true
   parent=FireModeTileRow
   widget=FireModeTile_AutoFire
   properties {
    customOption1=TapToShoot
    customOption2=FireButton
    customOptionsTabName=Add Custom Options
    fireModeDescription=Weapon fires automatically when you aim at an enemy
    presetFireModeType=AutoFire
    recommendedTag=Recommended
    videoUId=RJJdCpgQTFZzKyECcS
    fireModeArtImage {
     bIsDynamicallyLoaded=false
     drawAs=Image
     imageType=NoImage
     mirroring=NoMirror
     resourceName=None
     resourceObject=/FortniteCoreUI/Frontend/FireMode/art/MI_Bacchus_Firemode_Tile_Autofire
     tiling=NoTile
     imageSize=<576 720>
     margin=<ltrb 0 0 0 0>
     outlineSettings {
      bUseBrushTransparency=false
      roundingType=HalfHeightRadius
      width=0
      color {colorUseRule=UseColor_Specified, specifiedColor=<rgba 0 0 0 0>}
      cornerRadii=<0 0 0 1>
     }
     tintColor {colorUseRule=UseColor_Specified, specifiedColor=<rgba 1 1 1 1>}
     uVRegion {bIsValid=false, max=<0 0>, min=<0 0>}
    }
   }
   slot {
    class=HorizontalBoxSlot
    properties {horizontalAlignment=HAlign_Center, verticalAlignment=VAlign_Center, padding=<ltrb 50 0 50 0>, size {sizeRule=Automatic, value=0.33}}
   }
  }
  FireModeTile_DedicatedButton {
   widgetClassPath=&FireModeTile
   bIsVariable=true
   parent=FireModeTileRow
   widget=FireModeTile_DedicatedButton
   properties {
    customOption1=TapToShoot
    customOption2=AutoFire
    customOptionsTabName=Add Custom Options
    fireModeDescription=Only fire your weapon when you tap the fire button
    isRecommended=false
    presetFireModeType=FireButton
    recommendedTag=Recommended
    videoUId=IyhAtJSzsMmEJwVFyM
    fireModeArtImage {
     bIsDynamicallyLoaded=false
     drawAs=Image
     imageType=NoImage
     mirroring=NoMirror
     resourceName=None
     resourceObject=/FortniteCoreUI/Frontend/FireMode/art/MI_Bacchus_Firemode_Tile_Dedicated
     tiling=NoTile
     imageSize=<576 720>
     margin=<ltrb 0 0 0 0>
     outlineSettings {
      bUseBrushTransparency=false
      roundingType=HalfHeightRadius
      width=0
      color {colorUseRule=UseColor_Specified, specifiedColor=<rgba 0 0 0 0>}
      cornerRadii=<0 0 0 1>
     }
     tintColor {colorUseRule=UseColor_Specified, specifiedColor=<rgba 1 1 1 1>}
     uVRegion {bIsValid=false, max=<0 0>, min=<0 0>}
    }
   }
   slot {
    class=HorizontalBoxSlot
    properties {horizontalAlignment=HAlign_Center, verticalAlignment=VAlign_Center, size {sizeRule=Automatic, value=0.33}}
   }
  }
  DynamicRightPadding {
   widgetClassPath=&Spacer
   parent=FireModeTileRow
   widget=DynamicRightPadding
   slot {class=HorizontalBoxSlot, properties {size {sizeRule=Automatic, value=0.05}}}
  }
  CommonVisibilityWidgetLegacy_398 {
   widgetClassPath=&CommonVisibilityWidgetLegacy
   parent=contents
   widget=CommonVisibilityWidgetLegacy_398
   properties {
    bShowForMouseAndKeyboard=false
    bShowForTouch=false
    horizontalAlignment=HAlign_Center
    verticalAlignment=VAlign_Center
    padding=<ltrb 0 0 0 0>
   }
   slot {class=VerticalBoxSlot}
  }
  Footersafezone {
   widgetClassPath=&Overlay
   parent=VerticalBox_0
   widget=Footersafezone
   slot {class=VerticalBoxSlot, properties {padding=<ltrb 0 40 0 40>}}
  }
  SafeZone_1 {
   widgetClassPath=&SafeZone
   bIsVariable=true
   parent=Footersafezone
   widget=SafeZone_1
   properties {padBottom=false, padTop=false}
   slot {class=OverlaySlot, properties {horizontalAlignment=HAlign_Fill, verticalAlignment=VAlign_Bottom}}
  }
  HorizontalBox_3 {
   widgetClassPath=&HorizontalBox
   parent=SafeZone_1
   widget=HorizontalBox_3
   slot {class=SafeZoneSlot, properties {vAlign=VAlign_Bottom}}
  }
  OptionMessage {
   widgetClassPath=&Overlay
   parent=HorizontalBox_3
   widget=OptionMessage
   slot {class=HorizontalBoxSlot, properties {padding=<ltrb 50 0 0 0>, size {sizeRule=Fill, value=0.55}}}
  }
  HorizontalBox_2 {
   widgetClassPath=&HorizontalBox
   parent=OptionMessage
   widget=HorizontalBox_2
   slot {class=OverlaySlot, properties {verticalAlignment=VAlign_Center}}
  }
  settingsIcon {
   widgetClassPath=&Image
   bIsVariable=true
   parent=HorizontalBox_2
   widget=settingsIcon
   properties {brush {
     bIsDynamicallyLoaded=false
     drawAs=Image
     imageType=NoImage
     mirroring=NoMirror
     resourceName=None
     resourceObject=/Game/UI/Sidebar/Textures/T_UI_Icon_Sidebar_Settings
     tiling=NoTile
     imageSize=<32 32>
     margin=<ltrb 0 0 0 0>
     outlineSettings {
      bUseBrushTransparency=false
      roundingType=HalfHeightRadius
      width=0
      color {colorUseRule=UseColor_Specified, specifiedColor=<rgba 0 0 0 0>}
      cornerRadii=<0 0 0 0>
     }
     tintColor {colorUseRule=UseColor_Specified, specifiedColor=<rgba 0 0.7156 1 1>}
     uVRegion {bIsValid=false, max=<0 0>, min=<0 0>}
    }}
   slot {class=HorizontalBoxSlot, properties {horizontalAlignment=HAlign_Left, verticalAlignment=VAlign_Center, padding=<ltrb 0 0 10 0>}}
  }
  Text_QualityLevelCalculated {
   widgetClassPath=&CommonTextBlock
   bIsVariable=true
   parent=HorizontalBox_2
   widget=Text_QualityLevelCalculated
   properties {
    justification=Center
    text=The fire mode can be changed at any time in the Settings
    wrapTextAt=1000
    colorAndOpacity {colorUseRule=UseColor_Specified, specifiedColor=<rgba 0 0.7156 1 1>}
    font {
     bForceMonospaced=false
     bMaterialIsStencil=false
     fontMaterial=null
     fontObject=&BurbankSmall
     letterSpacing=0
     monospacedWidth=1
     size=20
     skewAmount=0
     typefaceFontName=Bold
     outlineSettings {
      bApplyOutlineToDropShadows=false
      bMiteredCorners=false
      bSeparateFillAlpha=false
      outlineMaterial=null
      outlineSize=0
      outlineColor=<rgba 0 0 0 1>
     }
    }
    renderTransform {angle=0, scale=<1 1>, shear=<-12 0>, translation=<0 0>}
    shadowOffset=<0 0>
   }
   slot {class=HorizontalBoxSlot}
  }
  Overlay_5 {
   widgetClassPath=&Overlay
   parent=HorizontalBox_3
   widget=Overlay_5
   slot {class=HorizontalBoxSlot, properties {verticalAlignment=VAlign_Center, size {sizeRule=Fill, value=0.4}}}
  }
  Button_Confirm {
   widgetClassPath=/FortUILibrary/Foundation/Buttons/IconTextButton_Leto
   bInherited=true
   bIsVariable=true
   parent=Overlay_5
   widget=Button_Confirm
   properties {
    buttonText=CONFIRM
    minHeight=90
    minWidth=300
    style=/FortUILibrary/Foundation/Buttons/ButtonStyles/Primary/ButtonStyle_HoldPurchase_Yellow.ButtonStyle_HoldPurchase_Yellow_C
    triggeringInputAction {dataTable=&InputActionDataTable, rowName=Confirm}
   }
   slot {class=OverlaySlot, properties {horizontalAlignment=HAlign_Right, verticalAlignment=VAlign_Center, padding=<ltrb 0 0 50 0>}}
  }
  MobileCloseButtonContainer {
   widgetClassPath=&SafeZone
   bInherited=true
   bIsVariable=true
   parent=Overlay_29
   widget=MobileCloseButtonContainer
   slot {class=OverlaySlot, properties {horizontalAlignment=HAlign_Right, padding=<ltrb 0 32 32 0>}}
  }
  VisibilityBorder_Touch {
   widgetClassPath=&CommonHardwareVisibilityBorder
   parent=MobileCloseButtonContainer
   widget=VisibilityBorder_Touch
   properties {visibilityQuery {
     autoDescription=" ANY( Input.Touch )"
     tokenStreamVersion=0
     userDescription=""
     queryTokenStream [0,1,1,1,0]
     tagDictionary [{tagName=Input.Touch}]
    }}
   slot {class=SafeZoneSlot}
  }
  Button_TouchClose {
   widgetClassPath=/FortUILibrary/Foundation/Buttons/CloseButton
   bInherited=true
   bIsVariable=true
   parent=VisibilityBorder_Touch
   widget=Button_TouchClose
   properties {
    onButtonBaseClicked=(&self:WidgetTree.Button_TouchClose.HandleOnButtonClicked)
    onButtonBaseDoubleClicked=(&self:WidgetTree.Button_TouchClose.HandleOnButtonDoubleClicked)
    onButtonBaseHovered=(&self:WidgetTree.Button_TouchClose.HandleOnButtonHovered)
    onButtonBaseUnhovered=(&self:WidgetTree.Button_TouchClose.HandleOnButtonUnhovered)
    onSelectedChangedBase=(&self:WidgetTree.Button_TouchClose.HandleOnSelectedChanged)
   }
   slot {class=BorderSlot, properties {padding=<ltrb 0 0 0 0>}}
  }
 ]
}
